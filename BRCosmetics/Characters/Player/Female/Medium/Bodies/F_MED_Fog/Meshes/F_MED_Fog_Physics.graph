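# Physics asset for the F_MED_Fog female player body cosmetic, covering the dangling knots, bows and ties on its shirt.
# Small capsules on the dyn_shirt bones form two branching constraint chains, with capsules on pelvis, spine and neck as body volumes.
# The bottom knot hangs from spine_02 and carries left and right ties.
# The top knot is angularly locked to neck_01 and carries both bows and ties.
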PhysicsAsset /BRCosmetics/Characters/Player/Female/Medium/Bodies/F_MED_Fog/Meshes/F_MED_Fog_Physics
bodies {
 dyn_shirt_bot_knot [{shapeType=Capsule, length=1, radius=1, shapeName=dyn_shirt_bot_knot_capsule, rotation=<rot -70 -0.0003 0.0003>}]
 dyn_shirt_bot_tie_l_01 [{shapeType=Capsule, length=1, radius=1, shapeName=dyn_shirt_bot_tie_l_01_capsule, rotation=<rot 90 -2.0836 -2.0836>}]
 dyn_shirt_bot_tie_l_02 [{shapeType=Capsule, length=1, radius=1, shapeName=dyn_shirt_bot_tie_l_02_capsule, rotation=<rot 90 -2.0836 -2.0836>}]
 dyn_shirt_bot_tie_r_01 [{shapeType=Capsule, length=1, radius=1, shapeName=dyn_shirt_bot_tie_r_01_capsule, rotation=<rot 90 11.1022 11.1022>}]
 dyn_shirt_bot_tie_r_02 [{shapeType=Capsule, length=1, radius=1, shapeName=dyn_shirt_bot_tie_r_02_capsule, rotation=<rot 90 11.1022 11.1022>}]
 dyn_shirt_top_bow_l_02 [{
   shapeType=Capsule
   length=1
   radius=1
   shapeName=dyn_shirt_top_bow_l_02_capsule
   center=<-1.0024 -0.529 0.3926>
   rotation=<rot 90 0 -10.0001>
  }]
 dyn_shirt_top_bow_r [{
   shapeType=Capsule
   length=1
   radius=1
   shapeName=dyn_shirt_top_bow_r_capsule
   center=<-0.0204 0.4712 0.6621>
   rotation=<rot 90 -7.125 -17.1249>
  }]
 dyn_shirt_top_knot [{
   shapeType=Capsule
   length=2
   radius=2
   shapeName=dyn_shirt_top_knot_capsule
   center=<0 -0.2336 0.196>
   rotation=<rot -29.4988 -28.3407 47.6059>
  }]
 dyn_shirt_top_tie_l_02 [{
   shapeType=Capsule
   length=1
   radius=1
   shapeName=dyn_shirt_top_tie_l_02_capsule
   center=<-0.0721 0.7329 0.6657>
   rotation=<rot 78.8801 -17.9352 -37.339>
  }]
 dyn_shirt_top_tie_l_03 [{
   shapeType=Capsule
   length=1
   radius=0.75
   shapeName=dyn_shirt_top_tie_l_03_capsule
   center=<0.0336 1.4681 0.5289>
   rotation=<rot 67.7313 -62.7268 -64.4941>
  }]
 dyn_shirt_top_tie_r_02 [{
   shapeType=Capsule
   length=1
   radius=1
   shapeName=dyn_shirt_top_tie_r_02_capsule
   center=<0.0344 1.0614 0.751>
   rotation=<rot 58.5252 -70.5745 -83.2598>
  }]
 dyn_shirt_top_tie_r_03 [{
   shapeType=Capsule
   length=1
   radius=1
   shapeName=dyn_shirt_top_tie_r_03_capsule
   center=<0.5718 0.8669 0.5371>
   rotation=<rot 48.9737 -78.308 -74.6599>
  }]
 neck_01 [{shapeType=Capsule, length=7.75, radius=4, shapeName=neck_01_capsule, rotation=<rot 90 -140.9061 -130.9055>}]
 pelvis [{shapeType=Capsule, length=9.2333, radius=11.1548, center=<-3.2897 -2.0766 0.008>, rotation=<rot -2.0151 -90 0.0615>}]
 spine_01 [{shapeType=Capsule, length=14.75, radius=6.25, shapeName=spine_01_capsule, center=<0.2541 -2.5318 0>}]
 spine_02 [{shapeType=Capsule, length=9, radius=8, shapeName=spine_02_capsule, center=<-3.5796 2.0248 0>}]
 spine_05 [{shapeType=Capsule, length=12, radius=10, shapeName=spine_05_capsule, center=<3.0254 1.9059 0>}]
}
constraints [
 {bone1Name=dyn_shirt_bot_knot, bone2Name=spine_02, swing1LimitDegrees=20, swing2LimitDegrees=20, twistLimitDegrees=5}
 {bone1Name=dyn_shirt_bot_tie_r_01, bone2Name=dyn_shirt_bot_knot, swing1LimitDegrees=20, swing2LimitDegrees=20, twistLimitDegrees=5}
 {bone1Name=dyn_shirt_bot_tie_r_02, bone2Name=dyn_shirt_bot_tie_r_01, swing1LimitDegrees=20, swing2LimitDegrees=20, twistLimitDegrees=5}
 {bone1Name=dyn_shirt_bot_tie_l_01, bone2Name=dyn_shirt_bot_knot, swing1LimitDegrees=20, swing2LimitDegrees=20, twistLimitDegrees=5}
 {bone1Name=dyn_shirt_bot_tie_l_02, bone2Name=dyn_shirt_bot_tie_l_01, swing1LimitDegrees=20, swing2LimitDegrees=20, twistLimitDegrees=5}
 {
  bone1Name=dyn_shirt_top_knot
  bone2Name=neck_01
  swing1LimitDegrees=45
  swing1Motion=Locked
  swing2LimitDegrees=45
  swing2Motion=Locked
  twistLimitDegrees=45
  twistMotion=Locked
 }
 {bone1Name=dyn_shirt_top_bow_l_02, bone2Name=dyn_shirt_top_knot, swing1LimitDegrees=20, swing2LimitDegrees=20, twistLimitDegrees=2}
 {bone1Name=dyn_shirt_top_bow_r, bone2Name=dyn_shirt_top_knot, swing1LimitDegrees=20, swing2LimitDegrees=20, twistLimitDegrees=2}
 {bone1Name=dyn_shirt_top_tie_r_02, bone2Name=dyn_shirt_top_knot, swing1LimitDegrees=20, swing2LimitDegrees=20, twistLimitDegrees=2}
 {bone1Name=dyn_shirt_top_tie_r_03, bone2Name=dyn_shirt_top_tie_r_02, swing1LimitDegrees=20, swing2LimitDegrees=20, twistLimitDegrees=2}
 {bone1Name=dyn_shirt_top_tie_l_02, bone2Name=dyn_shirt_top_knot, swing1LimitDegrees=20, swing2LimitDegrees=20, twistLimitDegrees=2}
 {bone1Name=dyn_shirt_top_tie_l_03, bone2Name=dyn_shirt_top_tie_l_02, swing1LimitDegrees=20, swing2LimitDegrees=20, twistLimitDegrees=2}
]
