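# Animation Blueprint for the F_MED_Fog female body cosmetic, adding secondary motion to belt, guns, shirt knots and zippers.
# The parent mesh's copied pose feeds a Control Rig, then weighted bone constraints.
# Two low-alpha RigidBody passes with opposite gravity follow, plus additive bone offsets and Anim Dynamics on guns and rear zippers.
# The EventGraph computes skydive, thigh and movement alphas, which the shown AnimGraph leaves unread.
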AnimBlueprint /BRCosmetics/Characters/Player/Female/Medium/Bodies/F_MED_Fog/Meshes/F_MED_Fog_AnimBP
target_skeleton=&/F_MED_Fog_Skeleton
graphs [
 AnimGraph {nodes [
   n1 = Control Rig(Source=Copy Pose From Mesh() {bUseAttachedParent=true}) {
    alphaBoolBlend {blendOption=Linear}
    controlRigAssetReference {blueprintRigClass=&/F_MED_Fog_ControlRig.F_MED_Fog_ControlRig_C}
   }
   n2 = "Constraint - dyn_belt_MSTR (PPPPPPP)"!(
    ComponentPose=Local To Component(LocalPose=@n1)
    Alpha=1
    ConstraintWeights_0=0.0698
    ConstraintWeights_1:float=0.0143
    ConstraintWeights_2:float=0
    ConstraintWeights_3:float=0.0421
    ConstraintWeights_4:float=0.8738
    ConstraintWeights_5:float=0
    ConstraintWeights_6:float=0
   ) -> (Pose) {
    alphaBoolBlend {blendOption=Linear}
    boneToModify {boneName=dyn_belt_MSTR}
    constraintSetup [
     {targetBone {boneName=spine_02}}
     {targetBone {boneName=spine_04}}
     {targetBone {boneName=spine_03}}
     {targetBone {boneName=spine_01}}
     {targetBone {boneName=pelvis}}
     {targetBone {boneName=thigh_twist_01_l}}
     {targetBone {boneName=thigh_twist_01_r}}
    ]
    constraintWeights [1,1,1,1,1,1,1]
   }
   n3 = "Constraint - dyn_gun_r_MSTR (PPPPPPP)"!(
    ComponentPose=@n2
    Alpha=1
    ConstraintWeights_0=0
    ConstraintWeights_1:float=0
    ConstraintWeights_2:float=0
    ConstraintWeights_3:float=0
    ConstraintWeights_4:float=0.936
    ConstraintWeights_5:float=0.064
    ConstraintWeights_6:float=0
   ) -> (Pose) {
    alphaBoolBlend {blendOption=Linear}
    boneToModify {boneName=dyn_gun_r_MSTR}
    constraintSetup [
     {targetBone {boneName=spine_02}}
     {targetBone {boneName=spine_04}}
     {targetBone {boneName=spine_03}}
     {targetBone {boneName=spine_01}}
     {targetBone {boneName=pelvis}}
     {targetBone {boneName=thigh_twist_01_r}}
     {targetBone {boneName=thigh_r}}
    ]
    constraintWeights [1,1,1,1,1,1,1]
   }
   n4 = "Constraint - dyn_gun_l_MSTR (PPPPPPP)"!(
    ComponentPose=@n3
    Alpha=1
    ConstraintWeights_0=0
    ConstraintWeights_1:float=0
    ConstraintWeights_2:float=0
    ConstraintWeights_3:float=0.7267
    ConstraintWeights_4:float=0.2649
    ConstraintWeights_5:float=0
    ConstraintWeights_6:float=0.0084
   ) -> (Pose) {
    alphaBoolBlend {blendOption=Linear}
    boneToModify {boneName=dyn_gun_l_MSTR}
    constraintSetup [
     {targetBone {boneName=spine_02}}
     {targetBone {boneName=spine_04}}
     {targetBone {boneName=spine_01}}
     {targetBone {boneName=pelvis}}
     {targetBone {boneName=thigh_twist_01_l}}
     {targetBone {boneName=thigh_twist_01_r}}
     {targetBone {boneName=thigh_l}}
    ]
    constraintWeights [1,1,1,1,1,1,1]
   }
   n5 = "Constraint - dyn_shirt_bot_knot_MSTR (PPPPPP)"!(
    ComponentPose=@n4
    Alpha=1
    ConstraintWeights_0=0.3583
    ConstraintWeights_1:float=0.0614
    ConstraintWeights_2:float=0.1438
    ConstraintWeights_3:float=0.0002
    ConstraintWeights_4:float=0.2449
    ConstraintWeights_5:float=0.1914
   ) -> (Pose) {
    alphaBoolBlend {blendOption=Linear}
    boneToModify {boneName=dyn_shirt_bot_knot_MSTR}
    constraintSetup [
     {targetBone {boneName=spine_02}}
     {targetBone {boneName=spine_04}}
     {targetBone {boneName=spine_03}}
     {targetBone {boneName=spine_05}}
     {targetBone {boneName=spine_01}}
     {targetBone {boneName=pelvis}}
    ]
    constraintWeights [1,1,1,1,1,1]
   }
   n6 = "Constraint - dyn_zipper_bck_r (PPPPPPP)"!(
    ComponentPose=@n5
    Alpha=1
    ConstraintWeights_0=0.0005
    ConstraintWeights_1:float=0
    ConstraintWeights_2:float=0.0015
    ConstraintWeights_3:float=0.9761
    ConstraintWeights_4:float=0
    ConstraintWeights_5:float=0.0218
    ConstraintWeights_6:float=0
   ) -> (Pose) {
    alphaBoolBlend {blendOption=Linear}
    boneToModify {boneName=dyn_zipper_bck_r}
    constraintSetup [
     {targetBone {boneName=spine_02}}
     {targetBone {boneName=spine_03}}
     {targetBone {boneName=spine_01}}
     {targetBone {boneName=pelvis}}
     {targetBone {boneName=thigh_twist_01_l}}
     {targetBone {boneName=thigh_twist_01_r}}
     {targetBone {boneName=thigh_r}}
    ]
    constraintWeights [1,1,1,1,1,1,1]
   }
   n7 = "Constraint - dyn_zipper_bck_l_MSTR (PPPPPPP)"!(
    ComponentPose=@n6
    Alpha=1
    ConstraintWeights_0=0.0009
    ConstraintWeights_1:float=0
    ConstraintWeights_2:float=0.0019
    ConstraintWeights_3:float=0.9644
    ConstraintWeights_4:float=0.0324
    ConstraintWeights_5:float=0.0001
    ConstraintWeights_6:float=0.0004
   ) -> (Pose) {
    alphaBoolBlend {blendOption=Linear}
    boneToModify {boneName=dyn_zipper_bck_l_MSTR}
    constraintSetup [
     {targetBone {boneName=spine_02}}
     {targetBone {boneName=spine_03}}
     {targetBone {boneName=spine_01}}
     {targetBone {boneName=pelvis}}
     {targetBone {boneName=thigh_twist_01_l}}
     {targetBone {boneName=thigh_twist_01_r}}
     {targetBone {boneName=thigh_l}}
    ]
    constraintWeights [1,1,1,1,1,1,1]
   }
   n8 = "Constraint - dyn_shirt_top_knot (PPPPPP)"!(
    ComponentPose=@n7
    Alpha=1
    ConstraintWeights_0=1
    ConstraintWeights_1:float=0
    ConstraintWeights_2:float=0
    ConstraintWeights_3:float=0
    ConstraintWeights_4:float=0
    ConstraintWeights_5:float=0
   ) -> (Pose) {
    alphaBoolBlend {blendOption=Linear}
    boneToModify {boneName=dyn_shirt_top_knot}
    constraintSetup [
     {targetBone {boneName=spine_05}}
     {targetBone {boneName=clavicle_l}}
     {targetBone {boneName=clavicle_r}}
     {targetBone {boneName=neck_01}}
     {targetBone {boneName=head}}
     {targetBone {boneName=neck_02}}
    ]
    constraintWeights [1,1,1,1,1,1]
   }
   n9 = "Transform (Modify) Bone - Bone: dyn_shirt_top_knot_MSTR"(ComponentPose=@n8, Alpha=1, Translation="-0.000,-1,0", Rotation="0,0,0", Scale="1,1,1") {translationMode=BMM_Additive, alphaBoolBlend {blendOption=Linear}, boneToModify {boneName=dyn_shirt_top_knot_MSTR}, translation {y=-1}}
   n10 = RigidBody(ComponentPose=@n9, Alpha=0.2) {
    bClampLinearTranslationLimitToRefPose=true
    bOverrideWorldGravity=true
    overridePhysicsAsset=&/F_MED_Fog_Physics
    alphaBoolBlend {blendOption=Linear}
    baseBoneRef {boneName=spine_05}
    componentLinearAccScale=<0.1 0.1 0.1>
    overrideWorldGravity {z=-980}
    simSpaceSettings {externalLinearDragV=<0.2 0.2 0.2>}
   }
   n11 = RigidBody(ComponentPose=@n10, Alpha=0.2) {
    bClampLinearTranslationLimitToRefPose=true
    bOverrideWorldGravity=true
    overridePhysicsAsset=&F_MED_Fog_Physics
    alphaBoolBlend {blendOption=Linear}
    overrideWorldGravity {z=980}
    simSpaceSettings {externalLinearDragV=<0.2 0.2 0.2>}
   }
   n12 = "Transform (Modify) Bone - Bone: dyn_shirt_top_knot_MSTR"(ComponentPose=@n11, Alpha=1, Translation="0,0,0", Rotation="-0.000,0.000,20", Scale="1,1,1") {
    rotationMode=BMM_Additive
    translationMode=BMM_Additive
    alphaBoolBlend {blendOption=Linear}
    boneToModify {boneName=dyn_shirt_top_knot_MSTR}
    rotation {roll=20}
   }
   n13 = "Transform (Modify) Bone - Bone: dyn_shirt_top_knot_MSTR"(ComponentPose=@n12, Alpha=1, Translation="-0.000,0.000,0.826", Rotation="-0.000,0.000,60.000", Scale="1,1,1") {
    rotationMode=BMM_Additive
    translationMode=BMM_Additive
    alphaBoolBlend {blendOption=Linear}
    boneToModify {boneName=dyn_shirt_top_knot_MSTR}
    rotation {roll=60}
    translation {z=0.826}
   }
   n14 = "Transform (Modify) Bone - Bone: dyn_shirt_bot_knot_MSTR"(ComponentPose=@n13, Alpha=1, Translation="0,0,0", Rotation="-0.000,0.000,40.000", Scale="1,1,1") {
    rotationMode=BMM_Additive
    translationMode=BMM_Additive
    alphaBoolBlend {blendOption=Linear}
    boneToModify {boneName=dyn_shirt_bot_knot_MSTR}
    rotation {roll=40}
   }
   n15 = "Transform (Modify) Bone - Bone: dyn_shirt_top_knot_MSTR"(ComponentPose=@n14, Alpha=1, Translation="-0.000,0.000,1", Rotation="-0.000,0.000,10", Scale="1,1,1") {
    rotationMode=BMM_Additive
    translationMode=BMM_Additive
    alphaBoolBlend {blendOption=Linear}
    boneToModify {boneName=dyn_shirt_top_knot_MSTR}
    rotation {roll=10}
    translation {z=1}
   }
   n16 = Anim Dynamics - Bone: dyn_gun_l(ComponentPose=@n15, Alpha=1, ExternalForce="0,0,0") {
    bGravityOverrideInSimSpace=true
    bUseGravityOverride=true
    simulationSpace=BoneRelative
    alphaBoolBlend {blendOption=Linear}
    boundBone {boneName=dyn_gun_l}
    componentLinearAccScale=<0.5 0.5 0.5>
    gravityOverride {z=-980}
    physicsBodyDefinitions [{
      boundBone {boneName=dyn_gun_l}
      boxExtents=<20 10 10>
      constraintSetup {angularLimitsMax=<0 0 0>, angularLimitsMin=<0 0 0>, angularTarget=<0 0 0>}
      localJointOffset=<0 0 0>
     }]
    relativeSpaceBone {boneName=pelvis}
   }
   n17 = Anim Dynamics - Bone: dyn_gun_r(ComponentPose=@n16, Alpha=1, ExternalForce="0,0,0") {
    bGravityOverrideInSimSpace=true
    bUseGravityOverride=true
    simulationSpace=BoneRelative
    alphaBoolBlend {blendOption=Linear}
    boundBone {boneName=dyn_gun_r}
    componentLinearAccScale=<0.5 0.5 0.5>
    gravityOverride {z=-980}
    physicsBodyDefinitions [{
      boundBone {boneName=dyn_gun_r}
      boxExtents=<20 10 10>
      constraintSetup {angularLimitsMax=<0 0 0>, angularLimitsMin=<0 0 0>, angularTarget=<0 0 0>}
      localJointOffset=<0 0 0>
     }]
    relativeSpaceBone {boneName=pelvis}
   }
   n18 = "Transform (Modify) Bone - Bone: dyn_gun_l_MSTR"(ComponentPose=@n17, Alpha=1, Translation="2,0.000,0.000", Rotation="20,0.000,0.000", Scale="1,1,1") {
    rotationMode=BMM_Additive
    translationMode=BMM_Additive
    alphaBoolBlend {blendOption=Linear}
    boneToModify {boneName=dyn_gun_l_MSTR}
    rotation {pitch=20}
    translation {x=2}
   }
   n19 = "Transform (Modify) Bone - Bone: dyn_gun_r_MSTR"(ComponentPose=@n18, Alpha=1, Translation="-0.5,-0.000,0.000", Rotation="-10.000,0.000,-0.000", Scale="1,1,1") {
    rotationMode=BMM_Additive
    translationMode=BMM_Additive
    alphaBoolBlend {blendOption=Linear}
    boneToModify {boneName=dyn_gun_r_MSTR}
    rotation {pitch=-10}
    translation {x=-0.5}
   }
   n20 = Anim Dynamics - Bone: dyn_zipper_bck_l(ComponentPose=@n19, Alpha=1, ExternalForce="0,0,0") {
    bGravityOverrideInSimSpace=true
    bUseGravityOverride=true
    simulationSpace=BoneRelative
    alphaBoolBlend {blendOption=Linear}
    boundBone {boneName=dyn_zipper_bck_l}
    gravityOverride {z=-980}
    physicsBodyDefinitions [{
      boundBone {boneName=dyn_zipper_bck_l}
      boxExtents=<5 5 5>
      constraintSetup {angularLimitsMax=<0 0 0>, angularLimitsMin=<0 0 0>, angularTarget=<0 0 0>}
      localJointOffset=<0 0 0>
     }]
    relativeSpaceBone {boneName=thigh_l}
   }
   n21 = Anim Dynamics - Bone: dyn_zipper_bck_r(ComponentPose=@n20, Alpha=1, ExternalForce="0,0,0") {
    bGravityOverrideInSimSpace=true
    bUseGravityOverride=true
    simulationSpace=BoneRelative
    alphaBoolBlend {blendOption=Linear}
    boundBone {boneName=dyn_zipper_bck_r}
    gravityOverride {z=-980}
    physicsBodyDefinitions [{
      boundBone {boneName=dyn_zipper_bck_r}
      boxExtents=<5 5 5>
      constraintSetup {angularLimitsMax=<0 0 0>, angularLimitsMin=<0 0 0>, angularTarget=<0 0 0>}
      localJointOffset=<0 0 0>
     }]
    relativeSpaceBone {boneName=thigh_r}
   }
   Output Pose(Result=Component To Local(ComponentPose=@n21)) {layerGroup=None}
  ]}
 EventGraph {nodes [
   Event BlueprintUpdateAnimation()
   > n1 = Branch(Condition=Get bIsSkydiveFloating())
   > Set SkydiveAlpha(SkydiveAlpha:double=0.8) -> (Output_Get:double)
   > n2 = Set NormalAlpha(NormalAlpha:double=0) -> (Output_Get:double)
   n3 = GetOwnerSkeletalMeshComponent()
   n4 = BreakRotator(InRot=BreakTransform(InTransform=GetDeltaTransformFromRefPose(self=@n3, BoneName=thigh_r, BaseName=pelvis)).Rotation)
   Set R Thigh Rot(execute=[@n9.then, @n2.then], R Thigh Rot:double=MapRangeClamped(Value="Absolute (Float)"(A=@n4.Pitch), InRangeA=0, InRangeB=100)) -> (Output_Get:double)
   n5 = GetOwnerSkeletalMeshComponent()
   n6 = BreakRotator(InRot=BreakTransform(InTransform=GetDeltaTransformFromRefPose(self=@n5, BoneName=thigh_l, BaseName=pelvis)).Rotation)
   > Set L Thigh Rot(L Thigh Rot:double=MapRangeClamped(Value="Absolute (Float)"(A=@n6.Pitch), InRangeA=0, InRangeB=100)) -> (Output_Get:double)
   > n7 = Set movingAlpha(movingAlpha:double=MapRangeClamped(Value=Get PawnSpeed2D(), InRangeA=0, InRangeB=500)) -> (Output_Get:double)
   n8 = Get bIsCrouching()
   > Set crouchMoving_alpha(crouchMoving_alpha:double=float * float(A=@n7, B="To Float (Boolean)"(InBool=@n8)) {is=title K2Node_PromotableOperator}) -> (Output_Get:double)
   Set SkydiveAlpha(execute=@n1.else, SkydiveAlpha=0)
   > n9 = Set NormalAlpha(NormalAlpha=0.3)
  ]}
]
variables [
 crouchMoving_alpha {type=number}
 L Thigh Rot {type=number}
 movingAlpha {type=number}
 NormalAlpha {type=number}
 R Thigh Rot {type=number}
 SkydiveAlpha {type=number}
]
