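# Control Rig for the F_MED_Fog face accessory that corrects the back braids as the head rotates, apparently when looking up.
# A spherical pose reader on the head bone writes head_SPR_lookup_alpha, with debug drawing left on.
# That alpha weights chained global-space rotation and translation sets on the second bone of all six back braids, blending them toward fixed values.
# Front braids are untouched, and the aim and upperarm alpha variables are unused.
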ControlRigBlueprint /BRCosmetics/Characters/Player/Female/Medium/Bodies/F_MED_Fog/Meshes/Parts/F_MED_Fog_FaceAcc_CtrlRig
graphs [
 RigVMFunctionLibrary {links [], nodes []}
 RigVMModel {nodes [
   RigUnit_BeginExecution = &RigUnit_BeginExecution()
   n1 = &RigUnit_SphericalPoseReader(
    ActiveRegionScaleFactors=(PositiveWidth=1,NegativeWidth=0.5,PositiveHeight=0.75,NegativeHeight=0.75)
    ActiveRegionSize=0.1
    Debug=(bDrawDebug=True,DebugScale=25,DebugSegments=20,DebugThickness=0.25)
    DriverAxis=(X=0,Y=1,Z=0)
    DriverItem=(Type=Bone,Name="head")
    FalloffRegionScaleFactors=(PositiveWidth=1,NegativeWidth=0.5,PositiveHeight=1,NegativeHeight=1)
    FalloffSize=0.2
    OptionalParentItem=(Type=Bone)
    RotationOffset=(X=90,Y=0,Z=-30)
    ExecutePin=@RigUnit_BeginExecution.ExecutePin
   )
   > RigVMVariableNode(Value=@n1.OutputParam) {variable=head_SPR_lookup_alpha}
   n2 = RigVMVariableNode() {variable=head_SPR_lookup_alpha}
   > RigUnit_SetRotation = &RigUnit_SetRotation(Item=(Type=Bone,Name="dyn_bck_braids_l_A_02"), Space=GlobalSpace, Value=(X=-0.0298,Y=-0.0819,Z=-0.3407,W=0.9361), Weight=@n2.Value)
   > &RigUnit_SetRotation(Item=(Type=Bone,Name="dyn_bck_braids_r_A_02"), Space=GlobalSpace, Value=(X=-0.0868,Y=-0.0076,Z=-0.9924,W=0.0868), Weight=@n2.Value)
   > &RigUnit_SetRotation(Item=(Type=Bone,Name="dyn_bck_braids_l_B_02"), Space=GlobalSpace, Value=(X=-0.0298,Y=-0.0819,Z=-0.3407,W=0.9361), Weight=@n2.Value)
   > &RigUnit_SetRotation(Item=(Type=Bone,Name="dyn_bck_braids_r_B_02"), Space=GlobalSpace, Value=(X=-0.0868,Y=-0.0076,Z=-0.9924,W=0.0868), Weight=@n2.Value)
   > &RigUnit_SetRotation(Item=(Type=Bone,Name="dyn_bck_braids_l_C_02"), Space=GlobalSpace, Value=(X=-0.0298,Y=-0.0819,Z=-0.3407,W=0.9361), Weight=@n2.Value)
   > &RigUnit_SetRotation(Item=(Type=Bone,Name="dyn_bck_braids_r_C_02"), Space=GlobalSpace, Value=(X=-0.0868,Y=-0.0076,Z=-0.9924,W=0.0868), Weight=@n2.Value)
   > RigUnit_SetTranslation = /Script/ControlRig.RigUnit_SetTranslation(Item=(Type=Bone,Name="dyn_bck_braids_l_A_02"), Space=GlobalSpace, Value=(X=0,Y=-10,Z=0), bPropagateToChildren=True, Weight=@n2.Value) {is=struct RigVMUnitNode, template="Set Transform::Execute(in Item,in Space,in bInitial,in Value,in Weight,in bPropagateToChildren)"}
   > &RigUnit_SetTranslation(Item=(Type=Bone,Name="dyn_bck_braids_l_B_02"), Space=GlobalSpace, Value=(X=0,Y=-10,Z=0), bPropagateToChildren=True, Weight=@n2.Value) {template="Set Transform::Execute(in Item,in Space,in bInitial,in Value,in Weight,in bPropagateToChildren)"}
   > &RigUnit_SetTranslation(Item=(Type=Bone,Name="dyn_bck_braids_l_C_02"), Space=GlobalSpace, Value=(X=0,Y=-10,Z=0), bPropagateToChildren=True, Weight=@n2.Value) {template="Set Transform::Execute(in Item,in Space,in bInitial,in Value,in Weight,in bPropagateToChildren)"}
   > &RigUnit_SetTranslation(Item=(Type=Bone,Name="dyn_bck_braids_r_A_02"), Space=GlobalSpace, Value=(X=0,Y=-10,Z=0), bPropagateToChildren=True, Weight=@n2.Value) {template="Set Transform::Execute(in Item,in Space,in bInitial,in Value,in Weight,in bPropagateToChildren)"}
   > &RigUnit_SetTranslation(Item=(Type=Bone,Name="dyn_bck_braids_r_B_02"), Space=GlobalSpace, Value=(X=0,Y=-10,Z=0), bPropagateToChildren=True, Weight=@n2.Value) {template="Set Transform::Execute(in Item,in Space,in bInitial,in Value,in Weight,in bPropagateToChildren)"}
   > &RigUnit_SetTranslation(Item=(Type=Bone,Name="dyn_bck_braids_r_C_02"), Space=GlobalSpace, Value=(X=0,Y=-10,Z=0), bPropagateToChildren=True, Weight=@n2.Value) {template="Set Transform::Execute(in Item,in Space,in bInitial,in Value,in Weight,in bPropagateToChildren)"}
  ]}
]
hierarchy [
 root {loc=<0 0 0>, rot=<0 0 0 1>}
 pelvis {parent=root, loc=<0 -0.5301 98.0571>, rot=<0.0127 0.707 -0.0127 -0.707>}
 spine_01 {parent=pelvis, loc=<6.1208 -0.9328 0>, rot=<0 0 0.0576 -0.9983>}
 spine_02 {parent=spine_01, loc=<6.6022 -0.0153 0>, rot=<0 0 0.0022 -1>}
 spine_03 {parent=spine_02, loc=<6.4919 -0.0105 0>, rot=<0 0 0.0015 -1>}
 spine_04 {parent=spine_03, loc=<6.9596 -0.007 0>, rot=<0 0 0.0009 -1>}
 spine_05 {parent=spine_04, loc=<6.9981 -0.0017 0>, rot=<0 0 -0.0042 -1>}
 clavicle_l {parent=spine_05, loc=<5.4831 -0.1053 -3.2946>, rot=<0.0938 -0.7388 0.0307 -0.6666>}
 upperarm_l {parent=clavicle_l, loc=<9.9486 0 0>, rot=<0.0371 -0.2992 -0.0705 -0.9509>}
 neck_01 {parent=spine_05, loc=<11.3138 0.1529 0>, rot=<0 0 -0.2003 -0.9797>}
 neck_02 {parent=neck_01, loc=<4.5627 0 0>, rot=<0 0 0.0504 -0.9987>}
 head {parent=neck_02, loc=<5.3914 0 0>, rot=<0 0 0.115 -0.9934>}
 dyn_face_acc_MSTR {parent=head, loc=<11.2452 0.2899 0>, rot=<0.5663 0.3963 0.6306 0.353>}
 dyn_fr_braid_r_MSTR {parent=dyn_face_acc_MSTR, loc=<6.6517 0 0>, rot=<0.054 -0.3418 -0.1108 -0.9317>}
 dyn_fr_braid_r_01 {parent=dyn_fr_braid_r_MSTR, loc=<3.9844 0 0>, rot=<0.4126 -0.2622 -0.2072 -0.8474>}
 dyn_fr_braid_r_02 {parent=dyn_fr_braid_r_01, loc=<9.1281 0 0>, rot=<0.8254 0.0129 -0.158 -0.5419>}
 dyn_fr_braid_r_03 {parent=dyn_fr_braid_r_02, loc=<9.9433 0 0>, rot=<0.9529 -0.0076 -0.251 0.1702>}
 dyn_fr_braid_r_04 {parent=dyn_fr_braid_r_03, loc=<9.2515 0 0>, rot=<0 0 0 1>}
 dyn_fr_braid_l_MSTR {parent=dyn_face_acc_MSTR, loc=<4.1418 -5.1984 -0.2573>, rot=<-0.2385 -0.289 0.4891 -0.7876>}
 dyn_fr_braid_l_01 {parent=dyn_fr_braid_l_MSTR, loc=<3.9843 0 0>, rot=<-0.4126 -0.2621 0.2072 -0.8474>}
 dyn_fr_braid_l_02 {parent=dyn_fr_braid_l_01, loc=<9.1281 0 0>, rot=<-0.8254 0.0129 0.158 -0.5418>}
 dyn_fr_braid_l_03 {parent=dyn_fr_braid_l_02, loc=<9.9433 0 0>, rot=<0.9529 0.0076 -0.251 -0.1702>}
 dyn_fr_braid_l_04 {parent=dyn_fr_braid_l_03, loc=<9.2516 0 0>, rot=<0 0 0 1>}
 dyn_bck_braids_l_MSTR {parent=dyn_face_acc_MSTR, loc=<-7.2834 -2.847 -2.0446>, rot=<0 0 0.9848 -0.1736>}
 dyn_bck_braids_l_tie {parent=dyn_bck_braids_l_MSTR, loc=<0 0 0>, rot=<-0.0067 -0.0705 0.0175 -0.9973>}
 dyn_bck_braids_l_A_01 {parent=dyn_bck_braids_l_tie, loc=<5.5802 2.6021 0>, rot=<0.0907 -0.5253 -0.1479 -0.8331>}
 dyn_bck_braids_l_A_02 {parent=dyn_bck_braids_l_A_01, loc=<12.5773 0 0>, rot=<-0.4094 -0.1118 0.115 -0.8982>}
 dyn_bck_braids_l_A_03 {parent=dyn_bck_braids_l_A_02, loc=<13.7639 0 0>, rot=<-0.0411 -0.0235 0.0052 -0.9989>}
 dyn_bck_braids_l_A_04 {parent=dyn_bck_braids_l_A_03, loc=<14.7433 0 0>, rot=<0 0 0 1>}
 dyn_bck_braids_l_B_01 {parent=dyn_bck_braids_l_tie, loc=<7.9719 0.5439 0.0988>, rot=<-0.1159 -0.569 0.1332 -0.8032>}
 dyn_bck_braids_l_B_02 {parent=dyn_bck_braids_l_B_01, loc=<12.2354 0 0>, rot=<0.4806 -0.1009 -0.0883 -0.8666>}
 dyn_bck_braids_l_B_03 {parent=dyn_bck_braids_l_B_02, loc=<13.5852 0 0>, rot=<-0.6474 0.0257 0.0648 -0.7589>}
 dyn_bck_braids_l_B_04 {parent=dyn_bck_braids_l_B_03, loc=<13.077 0 0>, rot=<0 0 0 1>}
 dyn_bck_braids_l_C_01 {parent=dyn_bck_braids_l_tie, loc=<6.432 0.0909 -2.6342>, rot=<0.1206 -0.5864 -0.1625 -0.7844>}
 dyn_bck_braids_l_C_02 {parent=dyn_bck_braids_l_C_01, loc=<10.3377 0 0>, rot=<-0.5919 -0.0689 0.0909 -0.7979>}
 dyn_bck_braids_l_C_03 {parent=dyn_bck_braids_l_C_02, loc=<13.051 0 0>, rot=<0.1826 -0.001 -0.0119 -0.9831>}
 dyn_bck_braids_l_C_04 {parent=dyn_bck_braids_l_C_03, loc=<14.2004 0 0>, rot=<0 0 0 1>}
 dyn_bck_braids_r_MSTR {parent=dyn_face_acc_MSTR, loc=<-2.2309 7.6173 -1.5268>, rot=<0 0 -0.7373 -0.6756>}
 dyn_bck_braids_r_tie {parent=dyn_bck_braids_r_MSTR, loc=<0 0 0>, rot=<0.049 -0.0244 0.0855 -0.9948>}
 dyn_bck_braids_r_A_01 {parent=dyn_bck_braids_r_tie, loc=<6.1576 0 0>, rot=<0.0218 -0.5334 -0.0311 -0.845>}
 dyn_bck_braids_r_A_02 {parent=dyn_bck_braids_r_A_01, loc=<12.5667 0 0>, rot=<0.4174 -0.1084 -0.1178 -0.8945>}
 dyn_bck_braids_r_A_03 {parent=dyn_bck_braids_r_A_02, loc=<13.7732 0 0>, rot=<0.0377 -0.0257 -0.0049 -0.9989>}
 dyn_bck_braids_r_A_04 {parent=dyn_bck_braids_r_A_03, loc=<14.7433 0 0>, rot=<0 0 0 1>}
 dyn_bck_braids_r_B_01 {parent=dyn_bck_braids_r_tie, loc=<7.455 2.8763 0.0988>, rot=<0.2408 -0.5294 -0.309 -0.7525>}
 dyn_bck_braids_r_B_02 {parent=dyn_bck_braids_r_B_01, loc=<12.2288 0 0>, rot=<-0.4904 -0.0976 0.0907 -0.8612>}
 dyn_bck_braids_r_B_03 {parent=dyn_bck_braids_r_B_02, loc=<13.589 0 0>, rot=<0.6505 0.024 -0.0665 -0.7562>}
 dyn_bck_braids_r_B_04 {parent=dyn_bck_braids_r_B_03, loc=<13.077 0 0>, rot=<0 0 0 1>}
 dyn_bck_braids_r_C_01 {parent=dyn_bck_braids_r_tie, loc=<5.8683 2.6359 -2.6342>, rot=<0.0092 -0.5985 -0.0112 -0.801>}
 dyn_bck_braids_r_C_02 {parent=dyn_bck_braids_r_C_01, loc=<10.3377 0 0>, rot=<0.592 -0.0689 -0.0909 -0.7979>}
 dyn_bck_braids_r_C_03 {parent=dyn_bck_braids_r_C_02, loc=<13.051 0 0>, rot=<-0.1827 -0.001 0.0119 -0.9831>}
 dyn_bck_braids_r_C_04 {parent=dyn_bck_braids_r_C_03, loc=<14.2004 0 0>, rot=<0 0 0 1>}
 faceAttach {parent=head, loc=<0.6201 0.9296 0>, rot=<0.0029 -0.7071 -0.0029 -0.7071>}
 clavicle_r {parent=spine_05, loc=<5.4831 -0.1053 3.2946>, rot=<0.7388 0.0938 -0.6666 -0.0307>}
 upperarm_r {parent=clavicle_r, loc=<-9.9486 0 0>, rot=<0.0371 -0.2992 -0.0705 -0.9509>}
]
variables [
 aim_lookup_SPR_alpha {type=double, default=0}
 head_SPR_lookup_alpha {type=double, default=0}
 upperarm_r_SPR_up_alpha {type=double, default=0}
]
