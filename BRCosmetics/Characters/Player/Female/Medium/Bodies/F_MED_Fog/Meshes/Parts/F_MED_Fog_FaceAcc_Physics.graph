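# Physics asset for the face accessory part of the medium female Fog body, defining capsule bodies and joint limits for front and back braids.
# Each side has a front chain rooted at a MSTR bone and three back chains sharing a tie bone.
# Swing and twist limits are tight, with swing loosening toward front braid tips.
# Unconstrained capsules cover the head, neck, spine, clavicles, upper arms and face accessory root, apparently as colliders.
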PhysicsAsset /BRCosmetics/Characters/Player/Female/Medium/Bodies/F_MED_Fog/Meshes/Parts/F_MED_Fog_FaceAcc_Physics
bodies {
 clavicle_l [{
   shapeType=Capsule
   length=5.75
   radius=6.5
   shapeName=clavicle_l_capsule
   center=<6.9856 1.1134 -2.9317>
   rotation=<rot 85.0044 -85.6667 -94.3325>
  }]
 clavicle_r [{shapeType=Capsule, length=5.75, radius=6.5, shapeName=clavicle_r_capsule, center=<-7.9231 -1.601 2.4255>, rotation=<rot -90 90 -90.0005>}]
 dyn_bck_braids_l_A_01 [{
   shapeType=Capsule
   length=2.5
   radius=2.5
   shapeName=dyn_bck_braids_l_A_01_capsule
   center=<3.5136 0.8378 0.6195>
   rotation=<rot -80 0.0004 -0.0001>
  }]
 dyn_bck_braids_l_A_02 [{
   shapeType=Capsule
   length=4.75
   radius=3
   shapeName=dyn_bck_braids_l_A_02_capsule
   center=<1.6188 0 0>
   rotation=<rot 90 -172.4476 -172.4478>
  }]
 dyn_bck_braids_l_A_03 [{shapeType=Capsule, length=5, radius=3, shapeName=dyn_bck_braids_l_A_03_capsule, rotation=<rot 90 -0.9002 -0.9002>}]
 dyn_bck_braids_l_A_04 [{shapeType=Capsule, length=3.5, radius=3, shapeName=dyn_bck_braids_l_A_04_capsule, center=<-2.7614 0 0>, rotation=<rot 90 0 0>}]
 dyn_bck_braids_l_B_01 [{
   shapeType=Capsule
   length=1.75
   radius=2.5
   shapeName=dyn_bck_braids_l_B_01_capsule
   center=<2.4167 0 0.4261>
   rotation=<rot -80.0003 0.0004 -0.0002>
  }]
 dyn_bck_braids_l_B_02 [{shapeType=Capsule, length=4.5, radius=2.75, shapeName=dyn_bck_braids_l_B_02_capsule, center=<0.3959 0.2369 0>, rotation=<rot 90 0 9.9991>}]
 dyn_bck_braids_l_B_03 [{shapeType=Capsule, length=5, radius=3, shapeName=dyn_bck_braids_l_B_03_capsule, rotation=<rot 90 -0.1095 -0.1095>}]
 dyn_bck_braids_l_B_04 [{
   shapeType=Capsule
   length=3.5
   radius=3
   shapeName=dyn_bck_braids_l_B_04_capsule
   center=<-1.8001 0 0>
   rotation=<rot 90 -179.9999 -179.9999>
  }]
 dyn_bck_braids_l_C_01 [{
   shapeType=Capsule
   length=2.5
   radius=2.25
   shapeName=dyn_bck_braids_l_C_01_capsule
   center=<1.1481 0 -0.0947>
   rotation=<rot -80 0.0006 -0.0001>
  }]
 dyn_bck_braids_l_C_02 [{
   shapeType=Capsule
   length=4.25
   radius=3
   shapeName=dyn_bck_braids_l_C_02_capsule
   center=<0.7162 -0.5596 0.6526>
   rotation=<rot 90 90 90.0001>
  }]
 dyn_bck_braids_l_C_03 [{
   shapeType=Capsule
   length=5
   radius=3
   shapeName=dyn_bck_braids_l_C_03_capsule
   center=<0 -0.291 0.2256>
   rotation=<rot 90 -179.8417 -179.8418>
  }]
 dyn_bck_braids_l_C_04 [{
   shapeType=Capsule
   length=3.5
   radius=3
   shapeName=dyn_bck_braids_l_C_04_capsule
   center=<-2.3666 0 0>
   rotation=<rot 90 -179.9993 -179.9995>
  }]
 dyn_bck_braids_l_tie [{shapeType=Capsule, length=1.75, radius=2.5, shapeName=dyn_bck_braids_l_tie_capsule}]
 dyn_bck_braids_r_A_01 [{
   shapeType=Capsule
   length=3.5
   radius=2.75
   shapeName=dyn_bck_braids_r_A_01_capsule
   center=<3.6801 0 0.6489>
   rotation=<rot -90 -141.1413 141.141>
  }]
 dyn_bck_braids_r_A_02 [{
   shapeType=Capsule
   length=4.5
   radius=2.75
   shapeName=dyn_bck_braids_r_A_02_capsule
   center=<1.9071 -0.2005 0.0011>
   rotation=<rot -88.2656 -94.9434 95.0944>
  }]
 dyn_bck_braids_r_A_03 [{shapeType=Capsule, length=5.25, radius=2.75, shapeName=dyn_bck_braids_r_A_03_capsule, rotation=<rot 90 0.8992 0.8992>}]
 dyn_bck_braids_r_A_04 [{shapeType=Capsule, length=3.75, radius=2.75, shapeName=dyn_bck_braids_r_A_04_capsule, center=<-2.6143 -0.6238 0>, rotation=<rot 90 180 -180>}]
 dyn_bck_braids_r_B_01 [{
   shapeType=Capsule
   length=2.5
   radius=2.75
   shapeName=dyn_bck_braids_r_B_01_capsule
   center=<2.5449 0 0.4487>
   rotation=<rot -80.0003 -0.0004 0.0002>
  }]
 dyn_bck_braids_r_B_02 [{shapeType=Capsule, length=5, radius=2.5, shapeName=dyn_bck_braids_r_B_02_capsule, center=<1.2996 0 0>, rotation=<rot 90 27.9505 27.951>}]
 dyn_bck_braids_r_B_03 [{
   shapeType=Capsule
   length=4.25
   radius=2.75
   shapeName=dyn_bck_braids_r_B_03_capsule
   center=<0 0.4896 0.456>
   rotation=<rot 90 0.1083 0.1083>
  }]
 dyn_bck_braids_r_B_04 [{shapeType=Capsule, length=2, radius=2.75, shapeName=dyn_bck_braids_r_B_04_capsule, center=<-2.3742 -0.0106 0.5746>, rotation=<rot 90 0 0>}]
 dyn_bck_braids_r_C_01 [{
   shapeType=Capsule
   length=2.75
   radius=2.5
   shapeName=dyn_bck_braids_r_C_01_capsule
   center=<0.9284 -0.705 0.1637>
   rotation=<rot -80.0001 -0.0009 10.0003>
  }]
 dyn_bck_braids_r_C_02 [{
   shapeType=Capsule
   length=4.75
   radius=2.75
   shapeName=dyn_bck_braids_r_C_02_capsule
   center=<0.5282 0.3859 0>
   rotation=<rot 90 -0.4289 -0.429>
  }]
 dyn_bck_braids_r_C_03 [{
   shapeType=Capsule
   length=5.5
   radius=2.75
   shapeName=dyn_bck_braids_r_C_03_capsule
   center=<0.3982 0.3398 0>
   rotation=<rot 90 179.8417 179.8418>
  }]
 dyn_bck_braids_r_C_04 [{
   shapeType=Capsule
   length=2.25
   radius=3.25
   shapeName=dyn_bck_braids_r_C_04_capsule
   center=<-2.9131 0 -0.2808>
   rotation=<rot 90 160.5288 160.5288>
  }]
 dyn_bck_braids_r_tie [{shapeType=Capsule, length=1.75, radius=2.5, shapeName=dyn_bck_braids_r_tie_capsule}]
 dyn_face_acc_MSTR [{shapeType=Capsule, length=5.75, radius=8, shapeName=dyn_face_acc_MSTR_capsule, center=<0 0 -7.5732>, rotation=<rot 0.0002 -20 0>}]
 dyn_fr_braid_l_01 [{
   shapeType=Capsule
   length=2
   radius=1.75
   shapeName=dyn_fr_braid_l_01_capsule
   center=<4.3877 0.1976 0.2038>
   rotation=<rot -90 180 -179.9997>
  }]
 dyn_fr_braid_l_02 [{
   shapeType=Capsule
   length=2.75
   radius=2.25
   shapeName=dyn_fr_braid_l_02_capsule
   center=<2.4167 0.0001 0.3586>
   rotation=<rot -90 -179.9802 179.9801>
  }]
 dyn_fr_braid_l_03 [{
   shapeType=Capsule
   length=4.75
   radius=2.25
   shapeName=dyn_fr_braid_l_03_capsule
   center=<1.2328 -0.0703 0.1719>
   rotation=<rot 80.0003 -0.0001 9.9995>
  }]
 dyn_fr_braid_l_04 [{shapeType=Capsule, length=0.75, radius=2, shapeName=dyn_fr_braid_l_04_capsule, center=<-1.0898 0 0>, rotation=<rot 90 -160.5288 -160.5289>}]
 dyn_fr_braid_l_MSTR [{shapeType=Capsule, length=0.5, radius=2.25, shapeName=dyn_fr_braid_l_MSTR_capsule, rotation=<rot -39.9999 0.0001 -0.0001>}]
 dyn_fr_braid_r_01 [{
   shapeType=Capsule
   length=3
   radius=1.5
   shapeName=dyn_fr_braid_r_01_capsule
   center=<4.3877 -0.4119 0.2038>
   rotation=<rot -90 180 179.9997>
  }]
 dyn_fr_braid_r_02 [{
   shapeType=Capsule
   length=3.75
   radius=2
   shapeName=dyn_fr_braid_r_02_capsule
   center=<2.4167 0.1537 0.3586>
   rotation=<rot -90 0.1581 -0.158>
  }]
 dyn_fr_braid_r_03 [{
   shapeType=Capsule
   length=5.75
   radius=2
   shapeName=dyn_fr_braid_r_03_capsule
   center=<1.0153 0.2445 -0.1665>
   rotation=<rot 80.0005 0.0003 0.001>
  }]
 dyn_fr_braid_r_04 [{shapeType=Capsule, length=1.75, radius=1.75, shapeName=dyn_fr_braid_r_04_capsule, center=<-1.0898 0 0>, rotation=<rot 90 166.7374 166.7375>}]
 dyn_fr_braid_r_MSTR [{shapeType=Capsule, length=1.5, radius=2, shapeName=dyn_fr_braid_r_MSTR_capsule, rotation=<rot -39.9999 -0.0001 0.0001>}]
 head [{shapeType=Capsule, length=39, radius=7.25, shapeName=head_capsule, center=<0 0.8935 0>}]
 neck_02 [{shapeType=Capsule, length=5.25, radius=9.25, shapeName=neck_02_capsule, center=<-5.4106 0.9541 0>, rotation=<rot 0 -10.0005 0>}]
 spine_03 [{shapeType=Capsule, length=10, radius=10.5, shapeName=spine_03_capsule, rotation=<rot 90 177.4696 177.4697>}]
 spine_05 [{shapeType=Capsule, length=16, radius=6.75, shapeName=spine_05_capsule, center=<3.0044 1.394 0>}]
 upperarm_l [{shapeType=Capsule, length=10, radius=6.75, shapeName=upperarm_l_capsule, center=<6.5039 0 0>, rotation=<rot 90 -179.9901 -179.9902>}]
 upperarm_r [{shapeType=Capsule, length=10, radius=6.75, shapeName=upperarm_r_capsule, center=<-4.7787 0 0>, rotation=<rot 90 -160.5288 -160.529>}]
}
constraints [
 {bone1Name=dyn_fr_braid_r_02, bone2Name=dyn_fr_braid_r_01, swing1LimitDegrees=20, swing2LimitDegrees=20, twistLimitDegrees=5}
 {bone1Name=dyn_fr_braid_r_03, bone2Name=dyn_fr_braid_r_02, swing1LimitDegrees=45, swing2LimitDegrees=45, twistLimitDegrees=5}
 {bone1Name=dyn_fr_braid_r_04, bone2Name=dyn_fr_braid_r_03, swing1LimitDegrees=45, swing2LimitDegrees=45, twistLimitDegrees=5}
 {bone1Name=dyn_fr_braid_l_02, bone2Name=dyn_fr_braid_l_01, swing1LimitDegrees=20, swing2LimitDegrees=20, twistLimitDegrees=5}
 {bone1Name=dyn_fr_braid_l_03, bone2Name=dyn_fr_braid_l_02, swing1LimitDegrees=45, swing2LimitDegrees=45, twistLimitDegrees=5}
 {bone1Name=dyn_fr_braid_l_04, bone2Name=dyn_fr_braid_l_03, swing1LimitDegrees=45, swing2LimitDegrees=45, twistLimitDegrees=5}
 {bone1Name=dyn_bck_braids_l_A_02, bone2Name=dyn_bck_braids_l_A_01, swing1LimitDegrees=15, swing2LimitDegrees=15, twistLimitDegrees=10}
 {bone1Name=dyn_bck_braids_l_A_03, bone2Name=dyn_bck_braids_l_A_02, swing1LimitDegrees=20, swing2LimitDegrees=20, twistLimitDegrees=5}
 {bone1Name=dyn_bck_braids_l_A_04, bone2Name=dyn_bck_braids_l_A_03, swing1LimitDegrees=20, swing2LimitDegrees=20, twistLimitDegrees=5}
 {bone1Name=dyn_bck_braids_l_B_02, bone2Name=dyn_bck_braids_l_B_01, swing1LimitDegrees=15, swing2LimitDegrees=15, twistLimitDegrees=10}
 {bone1Name=dyn_bck_braids_l_B_03, bone2Name=dyn_bck_braids_l_B_02, swing1LimitDegrees=20, swing2LimitDegrees=20, twistLimitDegrees=5}
 {bone1Name=dyn_bck_braids_l_B_04, bone2Name=dyn_bck_braids_l_B_03, swing1LimitDegrees=20, swing2LimitDegrees=20, twistLimitDegrees=5}
 {bone1Name=dyn_bck_braids_l_C_02, bone2Name=dyn_bck_braids_l_C_01, swing1LimitDegrees=15, swing2LimitDegrees=15, twistLimitDegrees=10}
 {bone1Name=dyn_bck_braids_l_C_03, bone2Name=dyn_bck_braids_l_C_02, swing1LimitDegrees=20, swing2LimitDegrees=20, twistLimitDegrees=5}
 {bone1Name=dyn_bck_braids_l_C_04, bone2Name=dyn_bck_braids_l_C_03, swing1LimitDegrees=20, swing2LimitDegrees=20, twistLimitDegrees=5}
 {bone1Name=dyn_bck_braids_r_A_02, bone2Name=dyn_bck_braids_r_A_01, swing1LimitDegrees=15, swing2LimitDegrees=15, twistLimitDegrees=10}
 {bone1Name=dyn_bck_braids_r_A_03, bone2Name=dyn_bck_braids_r_A_02, swing1LimitDegrees=20, swing2LimitDegrees=20, twistLimitDegrees=5}
 {bone1Name=dyn_bck_braids_r_A_04, bone2Name=dyn_bck_braids_r_A_03, swing1LimitDegrees=20, swing2LimitDegrees=20, twistLimitDegrees=5}
 {bone1Name=dyn_bck_braids_r_B_02, bone2Name=dyn_bck_braids_r_B_01, swing1LimitDegrees=15, swing2LimitDegrees=15, twistLimitDegrees=10}
 {bone1Name=dyn_bck_braids_r_B_03, bone2Name=dyn_bck_braids_r_B_02, swing1LimitDegrees=20, swing2LimitDegrees=20, twistLimitDegrees=5}
 {bone1Name=dyn_bck_braids_r_B_04, bone2Name=dyn_bck_braids_r_B_03, swing1LimitDegrees=20, swing2LimitDegrees=20, twistLimitDegrees=5}
 {bone1Name=dyn_bck_braids_r_C_02, bone2Name=dyn_bck_braids_r_C_01, swing1LimitDegrees=15, swing2LimitDegrees=15, twistLimitDegrees=10}
 {bone1Name=dyn_bck_braids_r_C_03, bone2Name=dyn_bck_braids_r_C_02, swing1LimitDegrees=20, swing2LimitDegrees=20, twistLimitDegrees=5}
 {bone1Name=dyn_bck_braids_r_C_04, bone2Name=dyn_bck_braids_r_C_03, swing1LimitDegrees=20, swing2LimitDegrees=20, twistLimitDegrees=5}
 {bone1Name=dyn_fr_braid_r_01, bone2Name=dyn_fr_braid_r_MSTR, swing1LimitDegrees=15, swing2LimitDegrees=15, twistLimitDegrees=4}
 {bone1Name=dyn_fr_braid_l_01, bone2Name=dyn_fr_braid_l_MSTR, swing1LimitDegrees=15, swing2LimitDegrees=15, twistLimitDegrees=4}
 {bone1Name=dyn_bck_braids_l_A_01, bone2Name=dyn_bck_braids_l_tie, swing1LimitDegrees=20, swing2LimitDegrees=20, twistLimitDegrees=5}
 {bone1Name=dyn_bck_braids_l_B_01, bone2Name=dyn_bck_braids_l_tie, swing1LimitDegrees=20, swing2LimitDegrees=20, twistLimitDegrees=5}
 {bone1Name=dyn_bck_braids_l_C_01, bone2Name=dyn_bck_braids_l_tie, swing1LimitDegrees=20, swing2LimitDegrees=20, twistLimitDegrees=5}
 {bone1Name=dyn_bck_braids_r_A_01, bone2Name=dyn_bck_braids_r_tie, swing1LimitDegrees=20, swing2LimitDegrees=20, twistLimitDegrees=5}
 {bone1Name=dyn_bck_braids_r_B_01, bone2Name=dyn_bck_braids_r_tie, swing1LimitDegrees=20, swing2LimitDegrees=20, twistLimitDegrees=5}
 {bone1Name=dyn_bck_braids_r_C_01, bone2Name=dyn_bck_braids_r_tie, swing1LimitDegrees=20, swing2LimitDegrees=20, twistLimitDegrees=5}
]
