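# Post-process style AnimBlueprint for the Fog female character's FaceAcc mesh, posing and simulating its front and back braids.
# A cached NullPose feeds a Control Rig, then RigidBody physics with overridden gravity and damped component motion.
# Weighted constraints pull back braids toward clavicle, neck and upper spine, followed by stacked additive bone offsets, some bool-blended.
# A parallel branch from the same cached pose repeats constraints and offsets without simulation.
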AnimBlueprint /BRCosmetics/Characters/Player/Female/Medium/Bodies/F_MED_Fog/Meshes/Parts/F_MED_Fog_FaceAcc_AnimBP
target_skeleton=&/F_MED_Fog_FaceAcc_Skeleton
graphs [
 AnimGraph {nodes [
   "Transform (Modify) Bone - Bone: dyn_fr_braid_r_02"(Alpha=1, Translation="0,0,0", Rotation="30.000,0.000,0.000", Scale="1,1,1") {
    rotationMode=BMM_Additive
    translationMode=BMM_Additive
    alphaBoolBlend {blendOption=Linear}
    boneToModify {boneName=dyn_fr_braid_r_02}
    rotation {pitch=30}
   }
   n1 = "Transform (Modify) Bone - Bone: dyn_bck_braids_r_A_01"(Alpha=1, Translation="0,-2,0.000", Rotation="0,0,0", Scale="1,1,1") {
    rotationMode=BMM_Additive
    translationMode=BMM_Additive
    alphaBoolBlend {blendOption=Linear}
    boneToModify {boneName=dyn_bck_braids_r_A_01}
    translation {y=-2}
   }
   n2 = "Transform (Modify) Bone - Bone: dyn_bck_braids_r_B_01"(ComponentPose=@n1, Alpha=1, Translation="0,-2,0", Rotation="0,0,0", Scale="1,1,1") {
    rotationMode=BMM_Additive
    translationMode=BMM_Additive
    alphaBoolBlend {blendOption=Linear}
    boneToModify {boneName=dyn_bck_braids_r_B_01}
    translation {y=-2}
   }
   n3 = "Transform (Modify) Bone - Bone: dyn_bck_braids_r_C_01"(ComponentPose=@n2, Alpha=1, Translation="0,-2,-0.000", Rotation="0,0,0", Scale="1,1,1") {
    rotationMode=BMM_Additive
    translationMode=BMM_Additive
    alphaBoolBlend {blendOption=Linear}
    boneToModify {boneName=dyn_bck_braids_r_C_01}
    translation {y=-2}
   }
   n4 = "Transform (Modify) Bone - Bone: dyn_bck_braids_r_A_02"(ComponentPose=@n3, Alpha=1, Translation="0,-10,0.000", Rotation="0.000,0.000,-20.000", Scale="1,1,1") {
    rotationMode=BMM_Additive
    translationMode=BMM_Additive
    alphaBoolBlend {blendOption=Linear}
    boneToModify {boneName=dyn_bck_braids_r_A_02}
    rotation {roll=-20}
    translation {y=-10}
   }
   n5 = "Transform (Modify) Bone - Bone: dyn_bck_braids_r_B_02"(ComponentPose=@n4, Alpha=1, Translation="0,-10,0", Rotation="0.000,0.000,-20.000", Scale="1,1,1") {
    rotationMode=BMM_Additive
    translationMode=BMM_Additive
    alphaBoolBlend {blendOption=Linear}
    boneToModify {boneName=dyn_bck_braids_r_B_02}
    rotation {roll=-20}
    translation {y=-10}
   }
   "Transform (Modify) Bone - Bone: dyn_bck_braids_r_C_02"(ComponentPose=@n5, Alpha=1, Translation="0,-10,-0.000", Rotation="0.000,0.000,-20.000", Scale="1,1,1") {
    rotationMode=BMM_Additive
    translationMode=BMM_Additive
    alphaBoolBlend {blendOption=Linear}
    boneToModify {boneName=dyn_bck_braids_r_C_02}
    rotation {roll=-20}
    translation {y=-10}
   }
   n6 = RigidBody(Alpha=0.7) {
    bClampLinearTranslationLimitToRefPose=true
    bOverrideWorldGravity=true
    overridePhysicsAsset=&/F_MED_Fog_FaceAcc_Physics
    simulationSpace=BaseBoneSpace
    alphaBoolBlend {blendOption=Linear}
    baseBoneRef {boneName=head}
    componentAppliedLinearAccClamp=<3000 3000 3000>
    componentLinearAccScale=<0.4 0.4 0.4>
    componentLinearVelScale=<0.1 0.1 0.1>
    overrideWorldGravity {z=980}
   }
   RigidBody(ComponentPose=@n6, Alpha=0.3) {
    bClampLinearTranslationLimitToRefPose=true
    bOverrideWorldGravity=true
    overridePhysicsAsset=&F_MED_Fog_FaceAcc_Physics
    alphaBoolBlend {blendOption=Linear}
    baseBoneRef {boneName=spine_05}
    componentLinearAccScale {x=0.6}
    componentLinearVelScale {y=0.1, z=0.1}
    overrideWorldGravity {z=-980}
    simSpaceSettings {externalLinearDragV {y=0.2, z=0.2}}
   }
   n7 = Control Rig(Source=Use cached pose 'NullPose'()) {
    alphaBoolBlend {blendOption=Linear}
    controlRigAssetReference {blueprintRigClass=&/F_MED_Fog_FaceAcc_CtrlRig.F_MED_Fog_FaceAcc_CtrlRig_C}
   }
   n8 = RigidBody!(ComponentPose=Local To Component(LocalPose=@n7), Alpha, OverrideWorldGravity:Vector="0,0,-980", ExternalForce) -> (Pose) {
    bClampLinearTranslationLimitToRefPose=true
    bOverrideWorldGravity=true
    overridePhysicsAsset=&F_MED_Fog_FaceAcc_Physics
    simulationSpace=BaseBoneSpace
    alphaBoolBlend {blendOption=Linear}
    baseBoneRef {boneName=root}
    componentLinearAccScale=<0.9 0.7 0.7>
    componentLinearVelScale {y=0.1, z=0.1}
    overrideWorldGravity {z=-980}
    simSpaceSettings {
     maxAngularAcceleration=1000
     maxAngularVelocity=3
     maxLinearAcceleration=700
     maxLinearVelocity=500
     worldAlpha=0.3
     externalLinearDragV=<0.2 0.2>
    }
   }
   n9 = "Constraint - dyn_bck_braids_l_A_02 (PPPP)"!(
    ComponentPose=@n8
    Alpha=1
    ConstraintWeights_0=0.2
    ConstraintWeights_1:float=0
    ConstraintWeights_2:float=0.4
    ConstraintWeights_3:float=0.4
   ) -> (Pose) {
    alphaBoolBlend {blendOption=Linear}
    boneToModify {boneName=dyn_bck_braids_l_A_02}
    constraintSetup [{targetBone {boneName=clavicle_l}}, {targetBone {boneName=upperarm_l}}, {targetBone {boneName=neck_01}}, {targetBone {boneName=spine_05}}]
    constraintWeights [1,1,1,1]
   }
   n10 = "Constraint - dyn_bck_braids_l_B_02 (PPPP)"!(
    ComponentPose=@n9
    Alpha=1
    ConstraintWeights_0=0.2
    ConstraintWeights_1:float=0
    ConstraintWeights_2:float=0.4
    ConstraintWeights_3:float=0.4
   ) -> (Pose) {
    alphaBoolBlend {blendOption=Linear}
    boneToModify {boneName=dyn_bck_braids_l_B_02}
    constraintSetup [{targetBone {boneName=clavicle_l}}, {targetBone {boneName=upperarm_l}}, {targetBone {boneName=neck_01}}, {targetBone {boneName=spine_05}}]
    constraintWeights [1,1,1,1]
   }
   n11 = "Constraint - dyn_bck_braids_l_C_02 (PPPP)"!(
    ComponentPose=@n10
    Alpha=1
    ConstraintWeights_0=0.2
    ConstraintWeights_1:float=0
    ConstraintWeights_2:float=0.4
    ConstraintWeights_3:float=0.4
   ) -> (Pose) {
    alphaBoolBlend {blendOption=Linear}
    boneToModify {boneName=dyn_bck_braids_l_C_02}
    constraintSetup [{targetBone {boneName=clavicle_l}}, {targetBone {boneName=upperarm_l}}, {targetBone {boneName=neck_01}}, {targetBone {boneName=spine_05}}]
    constraintWeights [1,1,1,1]
   }
   n12 = "Constraint - dyn_bck_braids_r_A_02 (PPPP)"!(
    ComponentPose=@n11
    Alpha=1
    ConstraintWeights_0=0.2
    ConstraintWeights_1:float=0
    ConstraintWeights_2:float=0.4
    ConstraintWeights_3:float=0.4
   ) -> (Pose) {
    alphaBoolBlend {blendOption=Linear}
    boneToModify {boneName=dyn_bck_braids_r_A_02}
    constraintSetup [{targetBone {boneName=clavicle_r}}, {targetBone {boneName=upperarm_r}}, {targetBone {boneName=neck_01}}, {targetBone {boneName=spine_05}}]
    constraintWeights [1,1,1,1]
   }
   n13 = "Constraint - dyn_bck_braids_r_B_02 (PPPP)"!(
    ComponentPose=@n12
    Alpha=1
    ConstraintWeights_0=0.2
    ConstraintWeights_1:float=0
    ConstraintWeights_2:float=0.4
    ConstraintWeights_3:float=0.4
   ) -> (Pose) {
    alphaBoolBlend {blendOption=Linear}
    boneToModify {boneName=dyn_bck_braids_r_B_02}
    constraintSetup [{targetBone {boneName=clavicle_r}}, {targetBone {boneName=upperarm_r}}, {targetBone {boneName=neck_01}}, {targetBone {boneName=spine_05}}]
    constraintWeights [1,1,1,1]
   }
   n14 = "Constraint - dyn_bck_braids_r_C_02 (PPPP)"!(
    ComponentPose=@n13
    Alpha=1
    ConstraintWeights_0=0.2
    ConstraintWeights_1:float=0
    ConstraintWeights_2:float=0.4
    ConstraintWeights_3:float=0.4
   ) -> (Pose) {
    alphaBoolBlend {blendOption=Linear}
    boneToModify {boneName=dyn_bck_braids_r_C_02}
    constraintSetup [{targetBone {boneName=clavicle_r}}, {targetBone {boneName=upperarm_r}}, {targetBone {boneName=neck_01}}, {targetBone {boneName=spine_05}}]
    constraintWeights [1,1,1,1]
   }
   n15 = "Transform (Modify) Bone - Bone: dyn_bck_braids_r_A_01"(ComponentPose=@n14, Alpha=1, Translation="0.000,-2,0.000", Rotation="0,0,0", Scale="1,1,1") {
    rotationMode=BMM_Additive
    translationMode=BMM_Additive
    alphaBoolBlend {blendOption=Linear}
    alphaScaleBiasClamp {bInterpResult=true}
    boneToModify {boneName=dyn_bck_braids_r_A_01}
    translation {y=-2}
   }
   n16 = "Transform (Modify) Bone - Bone: dyn_bck_braids_r_B_01"(ComponentPose=@n15, Alpha=1, Translation="0,-2,0", Rotation="0,0,0", Scale="1,1,1") {
    rotationMode=BMM_Additive
    translationMode=BMM_Additive
    alphaBoolBlend {blendOption=Linear}
    alphaScaleBiasClamp {bInterpResult=true}
    boneToModify {boneName=dyn_bck_braids_r_B_01}
    translation {y=-2}
   }
   n17 = "Transform (Modify) Bone - Bone: dyn_bck_braids_r_C_01"(ComponentPose=@n16, Alpha=1, Translation="0,-2,0", Rotation="0,0,0", Scale="1,1,1") {
    rotationMode=BMM_Additive
    translationMode=BMM_Additive
    alphaBoolBlend {blendOption=Linear}
    alphaScaleBiasClamp {bInterpResult=true}
    boneToModify {boneName=dyn_bck_braids_r_C_01}
    translation {y=-2}
   }
   n18 = "Transform (Modify) Bone - Bone: dyn_fr_braid_r_01"(ComponentPose=@n17, Alpha=1, Translation="0,0,0", Rotation="-0.000,-40.000,0.000", Scale="1,1,1") {
    rotationMode=BMM_Additive
    translationMode=BMM_Additive
    alphaBoolBlend {blendOption=Linear}
    alphaScaleBiasClamp {bInterpResult=true}
    boneToModify {boneName=dyn_fr_braid_r_01}
    rotation {yaw=-40}
   }
   n19 = "Transform (Modify) Bone - Bone: dyn_bck_braids_l_A_01"(ComponentPose=@n18, Alpha=1, Translation="0,-2,0", Rotation="0,0,0", Scale="1,1,1") {
    rotationMode=BMM_Additive
    translationMode=BMM_Additive
    alphaBoolBlend {blendOption=Linear}
    alphaScaleBiasClamp {bInterpResult=true}
    boneToModify {boneName=dyn_bck_braids_l_A_01}
    translation {y=-2}
   }
   n20 = "Transform (Modify) Bone - Bone: dyn_bck_braids_l_B_01"(ComponentPose=@n19, Alpha=1, Translation="0,-2,0", Rotation="0,0,0", Scale="1,1,1") {
    rotationMode=BMM_Additive
    translationMode=BMM_Additive
    alphaBoolBlend {blendOption=Linear}
    alphaScaleBiasClamp {bInterpResult=true}
    boneToModify {boneName=dyn_bck_braids_l_B_01}
    translation {y=-2}
   }
   n21 = "Transform (Modify) Bone - Bone: dyn_bck_braids_l_C_01"(ComponentPose=@n20, Alpha=1, Translation="0,-2,0", Rotation="0,0,0", Scale="1,1,1") {
    rotationMode=BMM_Additive
    translationMode=BMM_Additive
    alphaBoolBlend {blendOption=Linear}
    alphaScaleBiasClamp {bInterpResult=true}
    boneToModify {boneName=dyn_bck_braids_l_C_01}
    translation {y=-2}
   }
   n22 = "Transform (Modify) Bone - Bone: dyn_fr_braid_l_01"(ComponentPose=@n21, Alpha=1, Translation="0,0,0", Rotation="0.000,40.000,-0.000", Scale="1,1,1") {
    rotationMode=BMM_Additive
    translationMode=BMM_Additive
    alphaBoolBlend {blendOption=Linear}
    alphaScaleBiasClamp {bInterpResult=true}
    boneToModify {boneName=dyn_fr_braid_l_01}
    rotation {yaw=40}
   }
   n23 = "Transform (Modify) Bone - Bone: dyn_bck_braids_r_A_01"!(ComponentPose=@n22, bAlphaBoolEnabled:bool=True, Translation="0,0,0", Rotation="10,0.000,0", Scale="0.8,1,1") -> (Pose) {
    alphaInputType=Bool
    rotationMode=BMM_Additive
    rotationSpace=BCS_BoneSpace
    scaleMode=BMM_Additive
    scaleSpace=BCS_BoneSpace
    translationMode=BMM_Additive
    translationSpace=BCS_BoneSpace
    alphaBoolBlend {blendOption=Linear}
    alphaScaleBiasClamp {bInterpResult=true}
    boneToModify {boneName=dyn_bck_braids_r_A_01}
    rotation {pitch=10}
    scale {x=0.8}
   }
   n24 = "Transform (Modify) Bone - Bone: dyn_bck_braids_r_B_01"!(ComponentPose=@n23, bAlphaBoolEnabled:bool=True, Translation="0,0,0", Rotation="10,0,0", Scale="0.8,1,1") -> (Pose) {
    alphaInputType=Bool
    rotationMode=BMM_Additive
    rotationSpace=BCS_BoneSpace
    scaleMode=BMM_Additive
    scaleSpace=BCS_BoneSpace
    translationMode=BMM_Additive
    translationSpace=BCS_BoneSpace
    alphaBoolBlend {blendOption=Linear}
    alphaScaleBiasClamp {bInterpResult=true}
    boneToModify {boneName=dyn_bck_braids_r_B_01}
    rotation {pitch=10}
    scale {x=0.8}
   }
   n25 = "Transform (Modify) Bone - Bone: dyn_bck_braids_r_C_01"!(ComponentPose=@n24, bAlphaBoolEnabled:bool=True, Translation="0,0,0", Rotation="10,0,0", Scale="0.8,1,1") -> (Pose) {
    alphaInputType=Bool
    rotationMode=BMM_Additive
    rotationSpace=BCS_BoneSpace
    scaleMode=BMM_Additive
    scaleSpace=BCS_BoneSpace
    translationMode=BMM_Additive
    translationSpace=BCS_BoneSpace
    alphaBoolBlend {blendOption=Linear}
    alphaScaleBiasClamp {bInterpResult=true}
    boneToModify {boneName=dyn_bck_braids_r_C_01}
    rotation {pitch=10}
    scale {x=0.8}
   }
   n26 = "Transform (Modify) Bone - Bone: dyn_bck_braids_l_A_01"!(ComponentPose=@n25, bAlphaBoolEnabled:bool=True, Translation="0,0,0", Rotation="10,0.000,0", Scale="0.8,1,1") -> (Pose) {
    alphaInputType=Bool
    rotationMode=BMM_Additive
    rotationSpace=BCS_BoneSpace
    scaleMode=BMM_Additive
    scaleSpace=BCS_BoneSpace
    translationMode=BMM_Additive
    translationSpace=BCS_BoneSpace
    alphaBoolBlend {blendOption=Linear}
    alphaScaleBiasClamp {bInterpResult=true}
    boneToModify {boneName=dyn_bck_braids_l_A_01}
    rotation {pitch=10}
    scale {x=0.8}
   }
   n27 = "Transform (Modify) Bone - Bone: dyn_bck_braids_l_B_01"!(ComponentPose=@n26, bAlphaBoolEnabled:bool=True, Translation="0,0,0", Rotation="10,0,0", Scale="0.8,1,1") -> (Pose) {
    alphaInputType=Bool
    rotationMode=BMM_Additive
    rotationSpace=BCS_BoneSpace
    scaleMode=BMM_Additive
    scaleSpace=BCS_BoneSpace
    translationMode=BMM_Additive
    translationSpace=BCS_BoneSpace
    alphaBoolBlend {blendOption=Linear}
    alphaScaleBiasClamp {bInterpResult=true}
    boneToModify {boneName=dyn_bck_braids_l_B_01}
    rotation {pitch=10}
    scale {x=0.8}
   }
   n28 = "Transform (Modify) Bone - Bone: dyn_bck_braids_l_C_01"!(ComponentPose=@n27, bAlphaBoolEnabled:bool=True, Translation="0,0,0", Rotation="10,0.000,0", Scale="0.8,1,1") -> (Pose) {
    alphaInputType=Bool
    rotationMode=BMM_Additive
    rotationSpace=BCS_BoneSpace
    scaleMode=BMM_Additive
    scaleSpace=BCS_BoneSpace
    translationMode=BMM_Additive
    translationSpace=BCS_BoneSpace
    alphaBoolBlend {blendOption=Linear}
    alphaScaleBiasClamp {bInterpResult=true}
    boneToModify {boneName=dyn_bck_braids_l_C_01}
    rotation {pitch=10}
    scale {x=0.8}
   }
   n29 = "Transform (Modify) Bone - Bone: dyn_bck_braids_r_A_01"(ComponentPose=@n28, bAlphaBoolEnabled=True, Translation="0,0,0", Rotation="15,0,0", Scale="1,1,1") {
    alphaInputType=Bool
    rotationMode=BMM_Additive
    rotationSpace=BCS_BoneSpace
    scaleMode=BMM_Additive
    scaleSpace=BCS_BoneSpace
    translationMode=BMM_Additive
    translationSpace=BCS_BoneSpace
    alphaBoolBlend {blendInTime=0.5, blendOption=Linear, blendOutTime=0.3}
    alphaScaleBiasClamp {bInterpResult=true}
    boneToModify {boneName=dyn_bck_braids_r_A_01}
    rotation {pitch=15}
   }
   n30 = "Transform (Modify) Bone - Bone: dyn_bck_braids_r_B_01"(ComponentPose=@n29, bAlphaBoolEnabled=True, Translation="0,0,0", Rotation="15,0,0", Scale="1,1,1") {
    alphaInputType=Bool
    rotationMode=BMM_Additive
    rotationSpace=BCS_BoneSpace
    scaleMode=BMM_Additive
    scaleSpace=BCS_BoneSpace
    translationMode=BMM_Additive
    translationSpace=BCS_BoneSpace
    alphaBoolBlend {blendInTime=0.5, blendOption=Linear, blendOutTime=0.3}
    alphaScaleBiasClamp {bInterpResult=true}
    boneToModify {boneName=dyn_bck_braids_r_B_01}
    rotation {pitch=15}
   }
   n31 = "Transform (Modify) Bone - Bone: dyn_bck_braids_r_C_01"(ComponentPose=@n30, bAlphaBoolEnabled=True, Translation="0,0,0", Rotation="15,0,0", Scale="1,1,1") {
    alphaInputType=Bool
    rotationMode=BMM_Additive
    rotationSpace=BCS_BoneSpace
    scaleMode=BMM_Additive
    scaleSpace=BCS_BoneSpace
    translationMode=BMM_Additive
    translationSpace=BCS_BoneSpace
    alphaBoolBlend {blendInTime=0.5, blendOption=Linear, blendOutTime=0.3}
    alphaScaleBiasClamp {bInterpResult=true}
    boneToModify {boneName=dyn_bck_braids_r_C_01}
    rotation {pitch=15}
   }
   n32 = "Transform (Modify) Bone - Bone: dyn_fr_braid_r_01"!(ComponentPose=@n31, bAlphaBoolEnabled:bool=True, Translation="0,0,0", Rotation="0,20.000,0.000", Scale="1,1,1") -> (Pose) {
    alphaInputType=Bool
    rotationMode=BMM_Additive
    rotationSpace=BCS_BoneSpace
    scaleMode=BMM_Additive
    scaleSpace=BCS_BoneSpace
    translationMode=BMM_Additive
    translationSpace=BCS_BoneSpace
    alphaBoolBlend {blendInTime=0.5, blendOption=Linear, blendOutTime=0.3}
    alphaScaleBiasClamp {bInterpResult=true}
    boneToModify {boneName=dyn_fr_braid_r_01}
    rotation {yaw=20}
   }
   n33 = "Transform (Modify) Bone - Bone: dyn_bck_braids_l_A_01"(ComponentPose=@n32, bAlphaBoolEnabled=True, Translation="0,0,0", Rotation="15,0,0", Scale="1,1,1") {
    alphaInputType=Bool
    rotationMode=BMM_Additive
    rotationSpace=BCS_BoneSpace
    scaleMode=BMM_Additive
    scaleSpace=BCS_BoneSpace
    translationMode=BMM_Additive
    translationSpace=BCS_BoneSpace
    alphaBoolBlend {blendInTime=0.5, blendOption=Linear, blendOutTime=0.3}
    alphaScaleBiasClamp {bInterpResult=true}
    boneToModify {boneName=dyn_bck_braids_l_A_01}
    rotation {pitch=15}
   }
   n34 = "Transform (Modify) Bone - Bone: dyn_bck_braids_l_B_01"(ComponentPose=@n33, bAlphaBoolEnabled=True, Translation="0,0,0", Rotation="15,0,0", Scale="1,1,1") {
    alphaInputType=Bool
    rotationMode=BMM_Additive
    rotationSpace=BCS_BoneSpace
    scaleMode=BMM_Additive
    scaleSpace=BCS_BoneSpace
    translationMode=BMM_Additive
    translationSpace=BCS_BoneSpace
    alphaBoolBlend {blendInTime=0.5, blendOption=Linear, blendOutTime=0.3}
    alphaScaleBiasClamp {bInterpResult=true}
    boneToModify {boneName=dyn_bck_braids_l_B_01}
    rotation {pitch=15}
   }
   n35 = "Transform (Modify) Bone - Bone: dyn_bck_braids_l_C_01"(ComponentPose=@n34, bAlphaBoolEnabled=True, Translation="0,0,0", Rotation="15,0,0", Scale="1,1,1") {
    alphaInputType=Bool
    rotationMode=BMM_Additive
    rotationSpace=BCS_BoneSpace
    scaleMode=BMM_Additive
    scaleSpace=BCS_BoneSpace
    translationMode=BMM_Additive
    translationSpace=BCS_BoneSpace
    alphaBoolBlend {blendInTime=0.5, blendOption=Linear, blendOutTime=0.3}
    alphaScaleBiasClamp {bInterpResult=true}
    boneToModify {boneName=dyn_bck_braids_l_C_01}
    rotation {pitch=15}
   }
   n36 = "Transform (Modify) Bone - Bone: dyn_fr_braid_l_01"!(ComponentPose=@n35, bAlphaBoolEnabled:bool=True, Translation="0,0,0", Rotation="0,-20.000,-0.000", Scale="1,1,1") -> (Pose) {
    alphaInputType=Bool
    rotationMode=BMM_Additive
    rotationSpace=BCS_BoneSpace
    scaleMode=BMM_Additive
    scaleSpace=BCS_BoneSpace
    translationMode=BMM_Additive
    translationSpace=BCS_BoneSpace
    alphaBoolBlend {blendInTime=0.5, blendOption=Linear, blendOutTime=0.3}
    alphaScaleBiasClamp {bInterpResult=true}
    boneToModify {boneName=dyn_fr_braid_l_01}
    rotation {yaw=-20}
   }
   n37 = "Constraint - dyn_bck_braids_l_A_02 (PPPP)"(ComponentPose=@n36, Alpha=1, ConstraintWeights_0=0.2, ConstraintWeights_1=0, ConstraintWeights_2=0.4, ConstraintWeights_3=0.4) {
    alphaBoolBlend {blendOption=Linear}
    boneToModify {boneName=dyn_bck_braids_l_A_02}
    constraintSetup [{targetBone {boneName=clavicle_l}}, {targetBone {boneName=upperarm_l}}, {targetBone {boneName=neck_01}}, {targetBone {boneName=spine_05}}]
    constraintWeights [1,1,1,1]
   }
   n38 = "Constraint - dyn_bck_braids_l_B_02 (PPPP)"(ComponentPose=@n37, Alpha=1, ConstraintWeights_0=0.2, ConstraintWeights_1=0, ConstraintWeights_2=0.4, ConstraintWeights_3=0.4) {
    alphaBoolBlend {blendOption=Linear}
    boneToModify {boneName=dyn_bck_braids_l_B_02}
    constraintSetup [{targetBone {boneName=clavicle_l}}, {targetBone {boneName=upperarm_l}}, {targetBone {boneName=neck_01}}, {targetBone {boneName=spine_05}}]
    constraintWeights [1,1,1,1]
   }
   n39 = "Constraint - dyn_bck_braids_l_C_02 (PPPP)"(ComponentPose=@n38, Alpha=1, ConstraintWeights_0=0.2, ConstraintWeights_1=0, ConstraintWeights_2=0.4, ConstraintWeights_3=0.4) {
    alphaBoolBlend {blendOption=Linear}
    boneToModify {boneName=dyn_bck_braids_l_C_02}
    constraintSetup [{targetBone {boneName=clavicle_l}}, {targetBone {boneName=upperarm_l}}, {targetBone {boneName=neck_01}}, {targetBone {boneName=spine_05}}]
    constraintWeights [1,1,1,1]
   }
   n40 = "Constraint - dyn_bck_braids_r_A_02 (PPPP)"(ComponentPose=@n39, Alpha=1, ConstraintWeights_0=0.2, ConstraintWeights_1=0, ConstraintWeights_2=0.4, ConstraintWeights_3=0.4) {
    alphaBoolBlend {blendOption=Linear}
    boneToModify {boneName=dyn_bck_braids_r_A_02}
    constraintSetup [{targetBone {boneName=clavicle_r}}, {targetBone {boneName=upperarm_r}}, {targetBone {boneName=neck_01}}, {targetBone {boneName=spine_05}}]
    constraintWeights [1,1,1,1]
   }
   n41 = "Constraint - dyn_bck_braids_r_B_02 (PPPP)"(ComponentPose=@n40, Alpha=1, ConstraintWeights_0=0.2, ConstraintWeights_1=0, ConstraintWeights_2=0.4, ConstraintWeights_3=0.4) {
    alphaBoolBlend {blendOption=Linear}
    boneToModify {boneName=dyn_bck_braids_r_B_02}
    constraintSetup [{targetBone {boneName=clavicle_r}}, {targetBone {boneName=upperarm_r}}, {targetBone {boneName=neck_01}}, {targetBone {boneName=spine_05}}]
    constraintWeights [1,1,1,1]
   }
   n42 = "Constraint - dyn_bck_braids_r_C_02 (PPPP)"(ComponentPose=@n41, Alpha=1, ConstraintWeights_0=0.2, ConstraintWeights_1=0, ConstraintWeights_2=0.4, ConstraintWeights_3=0.4) {
    alphaBoolBlend {blendOption=Linear}
    boneToModify {boneName=dyn_bck_braids_r_C_02}
    constraintSetup [{targetBone {boneName=clavicle_r}}, {targetBone {boneName=upperarm_r}}, {targetBone {boneName=neck_01}}, {targetBone {boneName=spine_05}}]
    constraintWeights [1,1,1,1]
   }
   n43 = "Transform (Modify) Bone - Bone: dyn_bck_braids_r_A_03"!(ComponentPose=@n42, bAlphaBoolEnabled:bool=True, Translation="0,3,4", Rotation="0,0,0", Scale="1,1,1") -> (Pose) {
    alphaInputType=Bool
    rotationMode=BMM_Additive
    rotationSpace=BCS_BoneSpace
    scaleMode=BMM_Additive
    scaleSpace=BCS_BoneSpace
    translationMode=BMM_Additive
    translationSpace=BCS_BoneSpace
    alphaBoolBlend {blendInTime=0.5, blendOption=Linear, blendOutTime=0.3}
    alphaScaleBiasClamp {bInterpResult=true}
    boneToModify {boneName=dyn_bck_braids_r_A_03}
    translation {y=3, z=4}
   }
   n44 = "Transform (Modify) Bone - Bone: dyn_bck_braids_r_B_03"!(ComponentPose=@n43, bAlphaBoolEnabled:bool=True, Translation="0,3,4", Rotation="0,0,0", Scale="1,1,1") -> (Pose) {
    alphaInputType=Bool
    rotationMode=BMM_Additive
    rotationSpace=BCS_BoneSpace
    scaleMode=BMM_Additive
    scaleSpace=BCS_BoneSpace
    translationMode=BMM_Additive
    translationSpace=BCS_BoneSpace
    alphaBoolBlend {blendInTime=0.5, blendOption=Linear, blendOutTime=0.3}
    alphaScaleBiasClamp {bInterpResult=true}
    boneToModify {boneName=dyn_bck_braids_r_B_03}
    translation {y=3, z=4}
   }
   n45 = "Transform (Modify) Bone - Bone: dyn_bck_braids_r_C_03"!(ComponentPose=@n44, bAlphaBoolEnabled:bool=True, Translation="0.000,3,4", Rotation="0,0,0", Scale="1,1,1") -> (Pose) {
    alphaInputType=Bool
    rotationMode=BMM_Additive
    rotationSpace=BCS_BoneSpace
    scaleMode=BMM_Additive
    scaleSpace=BCS_BoneSpace
    translationMode=BMM_Additive
    translationSpace=BCS_BoneSpace
    alphaBoolBlend {blendInTime=0.5, blendOption=Linear, blendOutTime=0.3}
    alphaScaleBiasClamp {bInterpResult=true}
    boneToModify {boneName=dyn_bck_braids_r_C_03}
    translation {y=3, z=4}
   }
   n46 = "Transform (Modify) Bone - Bone: dyn_bck_braids_r_A_02"!(ComponentPose=@n45, bAlphaBoolEnabled:bool=True, Translation="-0.206,-1.134,2.447", Rotation="0.007,-10.000,0.005", Scale="1,1,1") -> (Pose) {
    alphaInputType=Bool
    rotationMode=BMM_Additive
    rotationSpace=BCS_BoneSpace
    scaleMode=BMM_Additive
    scaleSpace=BCS_BoneSpace
    translationMode=BMM_Additive
    translationSpace=BCS_BoneSpace
    alphaBoolBlend {blendInTime=0.5, blendOption=Linear, blendOutTime=0.3}
    alphaScaleBiasClamp {bInterpResult=true}
    boneToModify {boneName=dyn_bck_braids_r_A_02}
    rotation=<rot 0.0069 -9.9999 0.005>
    translation=<-0.206 -1.134 2.447>
   }
   n47 = "Transform (Modify) Bone - Bone: dyn_bck_braids_r_B_02"!(ComponentPose=@n46, bAlphaBoolEnabled:bool=True, Translation="-0.077,-6.271,0.254", Rotation="20.000,0.066,0.016", Scale="1,1,1") -> (Pose) {
    alphaInputType=Bool
    rotationMode=BMM_Additive
    rotationSpace=BCS_BoneSpace
    scaleMode=BMM_Additive
    scaleSpace=BCS_BoneSpace
    translationMode=BMM_Additive
    translationSpace=BCS_BoneSpace
    alphaBoolBlend {blendInTime=0.5, blendOption=Linear, blendOutTime=0.3}
    alphaScaleBiasClamp {bInterpResult=true}
    boneToModify {boneName=dyn_bck_braids_r_B_02}
    rotation=<rot 20 0.0659 0.016>
    translation=<-0.077 -6.271 0.254>
   }
   n48 = "Transform (Modify) Bone - Bone: dyn_bck_braids_r_C_02"!(ComponentPose=@n47, bAlphaBoolEnabled:bool=True, Translation="0,0,0", Rotation="0.011,-20.001,-0.019", Scale="1,1,1") -> (Pose) {
    alphaInputType=Bool
    rotationMode=BMM_Additive
    rotationSpace=BCS_BoneSpace
    scaleMode=BMM_Additive
    scaleSpace=BCS_BoneSpace
    translationMode=BMM_Additive
    translationSpace=BCS_BoneSpace
    alphaBoolBlend {blendInTime=0.5, blendOption=Linear, blendOutTime=0.3}
    alphaScaleBiasClamp {bInterpResult=true}
    boneToModify {boneName=dyn_bck_braids_r_C_02}
    rotation=<rot 0.011 -20.001 -0.019>
   }
   n49 = "Transform (Modify) Bone - Bone: dyn_bck_braids_l_A_03"!(ComponentPose=@n48, bAlphaBoolEnabled:bool=True, Translation="0,-1,2", Rotation="0,0,0", Scale="1,1,1") -> (Pose) {
    alphaInputType=Bool
    rotationMode=BMM_Additive
    rotationSpace=BCS_BoneSpace
    scaleMode=BMM_Additive
    scaleSpace=BCS_BoneSpace
    translationMode=BMM_Additive
    translationSpace=BCS_BoneSpace
    alphaBoolBlend {blendInTime=0.5, blendOption=Linear, blendOutTime=0.3}
    alphaScaleBiasClamp {bInterpResult=true}
    boneToModify {boneName=dyn_bck_braids_l_A_03}
    translation {y=-1, z=2}
   }
   n50 = "Transform (Modify) Bone - Bone: dyn_bck_braids_l_B_03"!(ComponentPose=@n49, bAlphaBoolEnabled:bool=True, Translation="0,-1,2", Rotation="0,0,0", Scale="1,1,1") -> (Pose) {
    alphaInputType=Bool
    rotationMode=BMM_Additive
    rotationSpace=BCS_BoneSpace
    scaleMode=BMM_Additive
    scaleSpace=BCS_BoneSpace
    translationMode=BMM_Additive
    translationSpace=BCS_BoneSpace
    alphaBoolBlend {blendInTime=0.5, blendOption=Linear, blendOutTime=0.3}
    alphaScaleBiasClamp {bInterpResult=true}
    boneToModify {boneName=dyn_bck_braids_l_B_03}
    translation {y=-1, z=2}
   }
   n51 = "Transform (Modify) Bone - Bone: dyn_bck_braids_l_C_03"!(ComponentPose=@n50, bAlphaBoolEnabled:bool=True, Translation="0.000,-1,2", Rotation="0,0,0", Scale="1,1,1") -> (Pose) {
    alphaInputType=Bool
    rotationMode=BMM_Additive
    rotationSpace=BCS_BoneSpace
    scaleMode=BMM_Additive
    scaleSpace=BCS_BoneSpace
    translationMode=BMM_Additive
    translationSpace=BCS_BoneSpace
    alphaBoolBlend {blendInTime=0.5, blendOption=Linear, blendOutTime=0.3}
    alphaScaleBiasClamp {bInterpResult=true}
    boneToModify {boneName=dyn_bck_braids_l_C_03}
    translation {y=-1, z=2}
   }
   n52 = "Transform (Modify) Bone - Bone: dyn_bck_braids_l_A_02"!(ComponentPose=@n51, bAlphaBoolEnabled:bool=True, Translation="0,0,0", Rotation="0.100,9.998,-0.045", Scale="1,1,1") -> (Pose) {
    alphaInputType=Bool
    rotationMode=BMM_Additive
    rotationSpace=BCS_BoneSpace
    scaleMode=BMM_Additive
    scaleSpace=BCS_BoneSpace
    translationMode=BMM_Additive
    translationSpace=BCS_BoneSpace
    alphaBoolBlend {blendInTime=0.5, blendOption=Linear, blendOutTime=0.3}
    alphaScaleBiasClamp {bInterpResult=true}
    boneToModify {boneName=dyn_bck_braids_l_A_02}
    rotation=<rot 0.1 9.998 -0.045>
   }
   n53 = "Transform (Modify) Bone - Bone: dyn_bck_braids_l_B_02"!(ComponentPose=@n52, bAlphaBoolEnabled:bool=True, Translation="0,0,0", Rotation="10.004,-10.198,-0.042", Scale="1,1,1") -> (Pose) {
    alphaInputType=Bool
    rotationMode=BMM_Additive
    rotationSpace=BCS_BoneSpace
    scaleMode=BMM_Additive
    scaleSpace=BCS_BoneSpace
    translationMode=BMM_Additive
    translationSpace=BCS_BoneSpace
    alphaBoolBlend {blendInTime=0.5, blendOption=Linear, blendOutTime=0.3}
    alphaScaleBiasClamp {bInterpResult=true}
    boneToModify {boneName=dyn_bck_braids_l_B_02}
    rotation=<rot 10.004 -10.198 -0.042>
   }
   n54 = "Transform (Modify) Bone - Bone: dyn_bck_braids_l_C_02"!(ComponentPose=@n53, bAlphaBoolEnabled:bool=True, Translation="0,0,0", Rotation="9.813,20.209,1.772", Scale="1,1,1") -> (Pose) {
    alphaInputType=Bool
    rotationMode=BMM_Additive
    rotationSpace=BCS_BoneSpace
    scaleMode=BMM_Additive
    scaleSpace=BCS_BoneSpace
    translationMode=BMM_Additive
    translationSpace=BCS_BoneSpace
    alphaBoolBlend {blendInTime=0.5, blendOption=Linear, blendOutTime=0.3}
    alphaScaleBiasClamp {bInterpResult=true}
    boneToModify {boneName=dyn_bck_braids_l_C_02}
    rotation=<rot 9.813 20.209 1.772>
   }
   n55 = "Constraint - dyn_bck_braids_l_A_02 (PPPP)"!(
    ComponentPose=Local To Component(LocalPose=Use cached pose 'NullPose'())
    ConstraintWeights_0=0.2
    ConstraintWeights_1=0
    ConstraintWeights_2=0.4
    ConstraintWeights_3=0.4
   ) -> (Pose) {
    alphaBoolBlend {blendOption=Linear}
    boneToModify {boneName=dyn_bck_braids_l_A_02}
    constraintSetup [{targetBone {boneName=clavicle_l}}, {targetBone {boneName=upperarm_l}}, {targetBone {boneName=neck_01}}, {targetBone {boneName=spine_05}}]
    constraintWeights [1,1,1,1]
   }
   n56 = "Constraint - dyn_bck_braids_l_B_02 (PPPP)"!(ComponentPose=@n55, ConstraintWeights_0=0.2, ConstraintWeights_1=0, ConstraintWeights_2=0.4, ConstraintWeights_3=0.4) -> (Pose) {
    alphaBoolBlend {blendOption=Linear}
    boneToModify {boneName=dyn_bck_braids_l_B_02}
    constraintSetup [{targetBone {boneName=clavicle_l}}, {targetBone {boneName=upperarm_l}}, {targetBone {boneName=neck_01}}, {targetBone {boneName=spine_05}}]
    constraintWeights [1,1,1,1]
   }
   n57 = "Constraint - dyn_bck_braids_l_C_02 (PPPP)"!(ComponentPose=@n56, ConstraintWeights_0=0.2, ConstraintWeights_1=0, ConstraintWeights_2=0.4, ConstraintWeights_3=0.4) -> (Pose) {
    alphaBoolBlend {blendOption=Linear}
    boneToModify {boneName=dyn_bck_braids_l_C_02}
    constraintSetup [{targetBone {boneName=clavicle_l}}, {targetBone {boneName=upperarm_l}}, {targetBone {boneName=neck_01}}, {targetBone {boneName=spine_05}}]
    constraintWeights [1,1,1,1]
   }
   n58 = "Constraint - dyn_bck_braids_r_A_02 (PPPP)"!(ComponentPose=@n57, ConstraintWeights_0=0.2, ConstraintWeights_1=0, ConstraintWeights_2=0.4, ConstraintWeights_3=0.4) -> (Pose) {
    alphaBoolBlend {blendOption=Linear}
    boneToModify {boneName=dyn_bck_braids_r_A_02}
    constraintSetup [{targetBone {boneName=clavicle_r}}, {targetBone {boneName=upperarm_r}}, {targetBone {boneName=neck_01}}, {targetBone {boneName=spine_05}}]
    constraintWeights [1,1,1,1]
   }
   n59 = "Constraint - dyn_bck_braids_r_B_02 (PPPP)"!(ComponentPose=@n58, ConstraintWeights_0=0.2, ConstraintWeights_1=0, ConstraintWeights_2=0.4, ConstraintWeights_3=0.4) -> (Pose) {
    alphaBoolBlend {blendOption=Linear}
    boneToModify {boneName=dyn_bck_braids_r_B_02}
    constraintSetup [{targetBone {boneName=clavicle_r}}, {targetBone {boneName=upperarm_r}}, {targetBone {boneName=neck_01}}, {targetBone {boneName=spine_05}}]
    constraintWeights [1,1,1,1]
   }
   n60 = "Constraint - dyn_bck_braids_r_C_02 (PPPP)"!(ComponentPose=@n59, ConstraintWeights_0=0.2, ConstraintWeights_1=0, ConstraintWeights_2=0.4, ConstraintWeights_3=0.4) -> (Pose) {
    alphaBoolBlend {blendOption=Linear}
    boneToModify {boneName=dyn_bck_braids_r_C_02}
    constraintSetup [{targetBone {boneName=clavicle_r}}, {targetBone {boneName=upperarm_r}}, {targetBone {boneName=neck_01}}, {targetBone {boneName=spine_05}}]
    constraintWeights [1,1,1,1]
   }
   n61 = "Transform (Modify) Bone - Bone: dyn_bck_braids_r_A_01"!(ComponentPose=@n60, Alpha:float=1, Translation="0.000,-2,0.000", Rotation="0,0,0", Scale="1,1,1") -> (Pose) {
    rotationMode=BMM_Additive
    translationMode=BMM_Additive
    alphaBoolBlend {blendOption=Linear}
    alphaScaleBiasClamp {bInterpResult=true}
    boneToModify {boneName=dyn_bck_braids_r_A_01}
    translation {y=-2}
   }
   n62 = "Transform (Modify) Bone - Bone: dyn_bck_braids_r_B_01"!(ComponentPose=@n61, Alpha:float=1, Translation="0,-2,0", Rotation="0,0,0", Scale="1,1,1") -> (Pose) {
    rotationMode=BMM_Additive
    translationMode=BMM_Additive
    alphaBoolBlend {blendOption=Linear}
    alphaScaleBiasClamp {bInterpResult=true}
    boneToModify {boneName=dyn_bck_braids_r_B_01}
    translation {y=-2}
   }
   n63 = "Transform (Modify) Bone - Bone: dyn_bck_braids_r_C_01"!(ComponentPose=@n62, Alpha:float=1, Translation="0,-2,0", Rotation="0,0,0", Scale="1,1,1") -> (Pose) {
    rotationMode=BMM_Additive
    translationMode=BMM_Additive
    alphaBoolBlend {blendOption=Linear}
    alphaScaleBiasClamp {bInterpResult=true}
    boneToModify {boneName=dyn_bck_braids_r_C_01}
    translation {y=-2}
   }
   n64 = "Transform (Modify) Bone - Bone: dyn_fr_braid_r_01"!(ComponentPose=@n63, Alpha:float=1, Translation="0,0,0", Rotation="-0.000,-40.000,0.000", Scale="1,1,1") -> (Pose) {
    rotationMode=BMM_Additive
    translationMode=BMM_Additive
    alphaBoolBlend {blendOption=Linear}
    alphaScaleBiasClamp {bInterpResult=true}
    boneToModify {boneName=dyn_fr_braid_r_01}
    rotation {yaw=-40}
   }
   n65 = "Transform (Modify) Bone - Bone: dyn_bck_braids_l_A_01"!(ComponentPose=@n64, Alpha:float=1, Translation="0,-2,0", Rotation="0,0,0", Scale="1,1,1") -> (Pose) {
    rotationMode=BMM_Additive
    translationMode=BMM_Additive
    alphaBoolBlend {blendOption=Linear}
    alphaScaleBiasClamp {bInterpResult=true}
    boneToModify {boneName=dyn_bck_braids_l_A_01}
    translation {y=-2}
   }
   n66 = "Transform (Modify) Bone - Bone: dyn_bck_braids_l_B_01"!(ComponentPose=@n65, Alpha:float=1, Translation="0,-2,0", Rotation="0,0,0", Scale="1,1,1") -> (Pose) {
    rotationMode=BMM_Additive
    translationMode=BMM_Additive
    alphaBoolBlend {blendOption=Linear}
    alphaScaleBiasClamp {bInterpResult=true}
    boneToModify {boneName=dyn_bck_braids_l_B_01}
    translation {y=-2}
   }
   n67 = "Transform (Modify) Bone - Bone: dyn_bck_braids_l_C_01"!(ComponentPose=@n66, Alpha:float=1, Translation="0,-2,0", Rotation="0,0,0", Scale="1,1,1") -> (Pose) {
    rotationMode=BMM_Additive
    translationMode=BMM_Additive
    alphaBoolBlend {blendOption=Linear}
    alphaScaleBiasClamp {bInterpResult=true}
    boneToModify {boneName=dyn_bck_braids_l_C_01}
    translation {y=-2}
   }
   n68 = "Transform (Modify) Bone - Bone: dyn_fr_braid_l_01"!(ComponentPose=@n67, Alpha:float=1, Translation="0,0,0", Rotation="0.000,40.000,-0.000", Scale="1,1,1") -> (Pose) {
    rotationMode=BMM_Additive
    translationMode=BMM_Additive
    alphaBoolBlend {blendOption=Linear}
    alphaScaleBiasClamp {bInterpResult=true}
    boneToModify {boneName=dyn_fr_braid_l_01}
    rotation {yaw=40}
   }
   n69 = "Transform (Modify) Bone - Bone: dyn_bck_braids_r_A_01"(ComponentPose=@n68, Alpha=1, Translation="0,0,0", Rotation="10,0.000,0", Scale="0.8,1,1") {
    rotationMode=BMM_Additive
    rotationSpace=BCS_BoneSpace
    scaleMode=BMM_Additive
    scaleSpace=BCS_BoneSpace
    translationMode=BMM_Additive
    translationSpace=BCS_BoneSpace
    alphaBoolBlend {blendOption=Linear}
    boneToModify {boneName=dyn_bck_braids_r_A_01}
    rotation {pitch=10}
    scale {x=0.8}
   }
   n70 = "Transform (Modify) Bone - Bone: dyn_bck_braids_r_B_01"(ComponentPose=@n69, Alpha=1, Translation="0,0,0", Rotation="10,0,0", Scale="0.8,1,1") {
    rotationMode=BMM_Additive
    rotationSpace=BCS_BoneSpace
    scaleMode=BMM_Additive
    scaleSpace=BCS_BoneSpace
    translationMode=BMM_Additive
    translationSpace=BCS_BoneSpace
    alphaBoolBlend {blendOption=Linear}
    boneToModify {boneName=dyn_bck_braids_r_B_01}
    rotation {pitch=10}
    scale {x=0.8}
   }
   n71 = "Transform (Modify) Bone - Bone: dyn_bck_braids_r_C_01"(ComponentPose=@n70, Alpha=1, Translation="0,0,0", Rotation="10,0,0", Scale="0.8,1,1") {
    rotationMode=BMM_Additive
    rotationSpace=BCS_BoneSpace
    scaleMode=BMM_Additive
    scaleSpace=BCS_BoneSpace
    translationMode=BMM_Additive
    translationSpace=BCS_BoneSpace
    alphaBoolBlend {blendOption=Linear}
    boneToModify {boneName=dyn_bck_braids_r_C_01}
    rotation {pitch=10}
    scale {x=0.8}
   }
   n72 = "Transform (Modify) Bone - Bone: dyn_bck_braids_l_A_01"(ComponentPose=@n71, Alpha=1, Translation="0,0,0", Rotation="10,0.000,0", Scale="0.8,1,1") {
    rotationMode=BMM_Additive
    rotationSpace=BCS_BoneSpace
    scaleMode=BMM_Additive
    scaleSpace=BCS_BoneSpace
    translationMode=BMM_Additive
    translationSpace=BCS_BoneSpace
    alphaBoolBlend {blendOption=Linear}
    boneToModify {boneName=dyn_bck_braids_l_A_01}
    rotation {pitch=10}
    scale {x=0.8}
   }
   n73 = "Transform (Modify) Bone - Bone: dyn_bck_braids_l_B_01"(ComponentPose=@n72, Alpha=1, Translation="0,0,0", Rotation="10,0,0", Scale="0.8,1,1") {
    rotationMode=BMM_Additive
    rotationSpace=BCS_BoneSpace
    scaleMode=BMM_Additive
    scaleSpace=BCS_BoneSpace
    translationMode=BMM_Additive
    translationSpace=BCS_BoneSpace
    alphaBoolBlend {blendOption=Linear}
    boneToModify {boneName=dyn_bck_braids_l_B_01}
    rotation {pitch=10}
    scale {x=0.8}
   }
   n74 = "Transform (Modify) Bone - Bone: dyn_bck_braids_l_C_01"(ComponentPose=@n73, Alpha=1, Translation="0,0,0", Rotation="10,0.000,0", Scale="0.8,1,1") {
    rotationMode=BMM_Additive
    rotationSpace=BCS_BoneSpace
    scaleMode=BMM_Additive
    scaleSpace=BCS_BoneSpace
    translationMode=BMM_Additive
    translationSpace=BCS_BoneSpace
    alphaBoolBlend {blendOption=Linear}
    boneToModify {boneName=dyn_bck_braids_l_C_01}
    rotation {pitch=10}
    scale {x=0.8}
   }
   n75 = "Transform (Modify) Bone - Bone: dyn_bck_braids_r_A_01"!(ComponentPose=@n74, bAlphaBoolEnabled:bool=True, Translation="0,0,0", Rotation="15,0,0", Scale="1,1,1") -> (Pose) {
    alphaInputType=Bool
    rotationMode=BMM_Additive
    rotationSpace=BCS_BoneSpace
    scaleMode=BMM_Additive
    scaleSpace=BCS_BoneSpace
    translationMode=BMM_Additive
    translationSpace=BCS_BoneSpace
    alphaBoolBlend {blendInTime=0.5, blendOption=Linear, blendOutTime=0.3}
    alphaScaleBiasClamp {bInterpResult=true}
    boneToModify {boneName=dyn_bck_braids_r_A_01}
    rotation {pitch=15}
   }
   n76 = "Transform (Modify) Bone - Bone: dyn_bck_braids_r_B_01"!(ComponentPose=@n75, bAlphaBoolEnabled:bool=True, Translation="0,0,0", Rotation="15,0,0", Scale="1,1,1") -> (Pose) {
    alphaInputType=Bool
    rotationMode=BMM_Additive
    rotationSpace=BCS_BoneSpace
    scaleMode=BMM_Additive
    scaleSpace=BCS_BoneSpace
    translationMode=BMM_Additive
    translationSpace=BCS_BoneSpace
    alphaBoolBlend {blendInTime=0.5, blendOption=Linear, blendOutTime=0.3}
    alphaScaleBiasClamp {bInterpResult=true}
    boneToModify {boneName=dyn_bck_braids_r_B_01}
    rotation {pitch=15}
   }
   n77 = "Transform (Modify) Bone - Bone: dyn_bck_braids_r_C_01"!(ComponentPose=@n76, bAlphaBoolEnabled:bool=True, Translation="0,0,0", Rotation="15,0,0", Scale="1,1,1") -> (Pose) {
    alphaInputType=Bool
    rotationMode=BMM_Additive
    rotationSpace=BCS_BoneSpace
    scaleMode=BMM_Additive
    scaleSpace=BCS_BoneSpace
    translationMode=BMM_Additive
    translationSpace=BCS_BoneSpace
    alphaBoolBlend {blendInTime=0.5, blendOption=Linear, blendOutTime=0.3}
    alphaScaleBiasClamp {bInterpResult=true}
    boneToModify {boneName=dyn_bck_braids_r_C_01}
    rotation {pitch=15}
   }
   n78 = "Transform (Modify) Bone - Bone: dyn_fr_braid_r_01"!(ComponentPose=@n77, bAlphaBoolEnabled:bool=True, Translation="0,0,0", Rotation="0,20.000,0.000", Scale="1,1,1") -> (Pose) {
    alphaInputType=Bool
    rotationMode=BMM_Additive
    rotationSpace=BCS_BoneSpace
    scaleMode=BMM_Additive
    scaleSpace=BCS_BoneSpace
    translationMode=BMM_Additive
    translationSpace=BCS_BoneSpace
    alphaBoolBlend {blendInTime=0.5, blendOption=Linear, blendOutTime=0.3}
    alphaScaleBiasClamp {bInterpResult=true}
    boneToModify {boneName=dyn_fr_braid_r_01}
    rotation {yaw=20}
   }
   n79 = "Transform (Modify) Bone - Bone: dyn_bck_braids_l_A_01"!(ComponentPose=@n78, bAlphaBoolEnabled:bool=True, Translation="0,0,0", Rotation="15,0,0", Scale="1,1,1") -> (Pose) {
    alphaInputType=Bool
    rotationMode=BMM_Additive
    rotationSpace=BCS_BoneSpace
    scaleMode=BMM_Additive
    scaleSpace=BCS_BoneSpace
    translationMode=BMM_Additive
    translationSpace=BCS_BoneSpace
    alphaBoolBlend {blendInTime=0.5, blendOption=Linear, blendOutTime=0.3}
    alphaScaleBiasClamp {bInterpResult=true}
    boneToModify {boneName=dyn_bck_braids_l_A_01}
    rotation {pitch=15}
   }
   n80 = "Transform (Modify) Bone - Bone: dyn_bck_braids_l_B_01"!(ComponentPose=@n79, bAlphaBoolEnabled:bool=True, Translation="0,0,0", Rotation="15,0,0", Scale="1,1,1") -> (Pose) {
    alphaInputType=Bool
    rotationMode=BMM_Additive
    rotationSpace=BCS_BoneSpace
    scaleMode=BMM_Additive
    scaleSpace=BCS_BoneSpace
    translationMode=BMM_Additive
    translationSpace=BCS_BoneSpace
    alphaBoolBlend {blendInTime=0.5, blendOption=Linear, blendOutTime=0.3}
    alphaScaleBiasClamp {bInterpResult=true}
    boneToModify {boneName=dyn_bck_braids_l_B_01}
    rotation {pitch=15}
   }
   n81 = "Transform (Modify) Bone - Bone: dyn_bck_braids_l_C_01"!(ComponentPose=@n80, bAlphaBoolEnabled:bool=True, Translation="0,0,0", Rotation="15,0,0", Scale="1,1,1") -> (Pose) {
    alphaInputType=Bool
    rotationMode=BMM_Additive
    rotationSpace=BCS_BoneSpace
    scaleMode=BMM_Additive
    scaleSpace=BCS_BoneSpace
    translationMode=BMM_Additive
    translationSpace=BCS_BoneSpace
    alphaBoolBlend {blendInTime=0.5, blendOption=Linear, blendOutTime=0.3}
    alphaScaleBiasClamp {bInterpResult=true}
    boneToModify {boneName=dyn_bck_braids_l_C_01}
    rotation {pitch=15}
   }
   n82 = "Transform (Modify) Bone - Bone: dyn_fr_braid_l_01"!(ComponentPose=@n81, bAlphaBoolEnabled:bool=True, Translation="0,0,0", Rotation="0,-20.000,-0.000", Scale="1,1,1") -> (Pose) {
    alphaInputType=Bool
    rotationMode=BMM_Additive
    rotationSpace=BCS_BoneSpace
    scaleMode=BMM_Additive
    scaleSpace=BCS_BoneSpace
    translationMode=BMM_Additive
    translationSpace=BCS_BoneSpace
    alphaBoolBlend {blendInTime=0.5, blendOption=Linear, blendOutTime=0.3}
    alphaScaleBiasClamp {bInterpResult=true}
    boneToModify {boneName=dyn_fr_braid_l_01}
    rotation {yaw=-20}
   }
   n83 = "Constraint - dyn_bck_braids_l_A_02 (PPPP)"!(ComponentPose=@n82, Alpha:float=1, ConstraintWeights_0=0.2, ConstraintWeights_1=0, ConstraintWeights_2=0.4, ConstraintWeights_3=0.4) -> (Pose) {
    alphaBoolBlend {blendOption=Linear}
    boneToModify {boneName=dyn_bck_braids_l_A_02}
    constraintSetup [{targetBone {boneName=clavicle_l}}, {targetBone {boneName=upperarm_l}}, {targetBone {boneName=neck_01}}, {targetBone {boneName=spine_05}}]
    constraintWeights [1,1,1,1]
   }
   n84 = "Constraint - dyn_bck_braids_l_B_02 (PPPP)"!(ComponentPose=@n83, Alpha:float=1, ConstraintWeights_0=0.2, ConstraintWeights_1=0, ConstraintWeights_2=0.4, ConstraintWeights_3=0.4) -> (Pose) {
    alphaBoolBlend {blendOption=Linear}
    boneToModify {boneName=dyn_bck_braids_l_B_02}
    constraintSetup [{targetBone {boneName=clavicle_l}}, {targetBone {boneName=upperarm_l}}, {targetBone {boneName=neck_01}}, {targetBone {boneName=spine_05}}]
    constraintWeights [1,1,1,1]
   }
   n85 = "Constraint - dyn_bck_braids_l_C_02 (PPPP)"!(ComponentPose=@n84, Alpha:float=1, ConstraintWeights_0=0.2, ConstraintWeights_1=0, ConstraintWeights_2=0.4, ConstraintWeights_3=0.4) -> (Pose) {
    alphaBoolBlend {blendOption=Linear}
    boneToModify {boneName=dyn_bck_braids_l_C_02}
    constraintSetup [{targetBone {boneName=clavicle_l}}, {targetBone {boneName=upperarm_l}}, {targetBone {boneName=neck_01}}, {targetBone {boneName=spine_05}}]
    constraintWeights [1,1,1,1]
   }
   n86 = "Constraint - dyn_bck_braids_r_A_02 (PPPP)"!(ComponentPose=@n85, Alpha:float=1, ConstraintWeights_0=0.2, ConstraintWeights_1=0, ConstraintWeights_2=0.4, ConstraintWeights_3=0.4) -> (Pose) {
    alphaBoolBlend {blendOption=Linear}
    boneToModify {boneName=dyn_bck_braids_r_A_02}
    constraintSetup [{targetBone {boneName=clavicle_r}}, {targetBone {boneName=upperarm_r}}, {targetBone {boneName=neck_01}}, {targetBone {boneName=spine_05}}]
    constraintWeights [1,1,1,1]
   }
   n87 = "Constraint - dyn_bck_braids_r_B_02 (PPPP)"!(ComponentPose=@n86, Alpha:float=1, ConstraintWeights_0=0.2, ConstraintWeights_1=0, ConstraintWeights_2=0.4, ConstraintWeights_3=0.4) -> (Pose) {
    alphaBoolBlend {blendOption=Linear}
    boneToModify {boneName=dyn_bck_braids_r_B_02}
    constraintSetup [{targetBone {boneName=clavicle_r}}, {targetBone {boneName=upperarm_r}}, {targetBone {boneName=neck_01}}, {targetBone {boneName=spine_05}}]
    constraintWeights [1,1,1,1]
   }
   n88 = "Constraint - dyn_bck_braids_r_C_02 (PPPP)"!(ComponentPose=@n87, Alpha:float=1, ConstraintWeights_0=0.2, ConstraintWeights_1=0, ConstraintWeights_2=0.4, ConstraintWeights_3=0.4) -> (Pose) {
    alphaBoolBlend {blendOption=Linear}
    boneToModify {boneName=dyn_bck_braids_r_C_02}
    constraintSetup [{targetBone {boneName=clavicle_r}}, {targetBone {boneName=upperarm_r}}, {targetBone {boneName=neck_01}}, {targetBone {boneName=spine_05}}]
    constraintWeights [1,1,1,1]
   }
   n89 = "Transform (Modify) Bone - Bone: dyn_bck_braids_r_A_03"(ComponentPose=@n88, bAlphaBoolEnabled=True, Translation="0,3,4", Rotation="0,0,0", Scale="1,1,1") {
    alphaInputType=Bool
    rotationMode=BMM_Additive
    rotationSpace=BCS_BoneSpace
    scaleMode=BMM_Additive
    scaleSpace=BCS_BoneSpace
    translationMode=BMM_Additive
    translationSpace=BCS_BoneSpace
    alphaBoolBlend {blendInTime=0.5, blendOption=Linear, blendOutTime=0.3}
    alphaScaleBiasClamp {bInterpResult=true}
    boneToModify {boneName=dyn_bck_braids_r_A_03}
    translation {y=3, z=4}
   }
   n90 = "Transform (Modify) Bone - Bone: dyn_bck_braids_r_B_03"(ComponentPose=@n89, bAlphaBoolEnabled=True, Translation="0,3,4", Rotation="0,0,0", Scale="1,1,1") {
    alphaInputType=Bool
    rotationMode=BMM_Additive
    rotationSpace=BCS_BoneSpace
    scaleMode=BMM_Additive
    scaleSpace=BCS_BoneSpace
    translationMode=BMM_Additive
    translationSpace=BCS_BoneSpace
    alphaBoolBlend {blendInTime=0.5, blendOption=Linear, blendOutTime=0.3}
    alphaScaleBiasClamp {bInterpResult=true}
    boneToModify {boneName=dyn_bck_braids_r_B_03}
    translation {y=3, z=4}
   }
   n91 = "Transform (Modify) Bone - Bone: dyn_bck_braids_r_C_03"(ComponentPose=@n90, bAlphaBoolEnabled=True, Translation="0.000,3,4", Rotation="0,0,0", Scale="1,1,1") {
    alphaInputType=Bool
    rotationMode=BMM_Additive
    rotationSpace=BCS_BoneSpace
    scaleMode=BMM_Additive
    scaleSpace=BCS_BoneSpace
    translationMode=BMM_Additive
    translationSpace=BCS_BoneSpace
    alphaBoolBlend {blendInTime=0.5, blendOption=Linear, blendOutTime=0.3}
    alphaScaleBiasClamp {bInterpResult=true}
    boneToModify {boneName=dyn_bck_braids_r_C_03}
    translation {y=3, z=4}
   }
   n92 = "Transform (Modify) Bone - Bone: dyn_bck_braids_r_A_02"(ComponentPose=@n91, bAlphaBoolEnabled=True, Translation="-0.206,-1.134,2.447", Rotation="0.007,-10.000,0.005", Scale="1,1,1") {
    alphaInputType=Bool
    rotationMode=BMM_Additive
    rotationSpace=BCS_BoneSpace
    scaleMode=BMM_Additive
    scaleSpace=BCS_BoneSpace
    translationMode=BMM_Additive
    translationSpace=BCS_BoneSpace
    alphaBoolBlend {blendInTime=0.5, blendOption=Linear, blendOutTime=0.3}
    alphaScaleBiasClamp {bInterpResult=true}
    boneToModify {boneName=dyn_bck_braids_r_A_02}
    rotation=<rot 0.0069 -9.9999 0.005>
    translation=<-0.206 -1.134 2.447>
   }
   n93 = "Transform (Modify) Bone - Bone: dyn_bck_braids_r_B_02"(ComponentPose=@n92, bAlphaBoolEnabled=True, Translation="-0.077,-6.271,0.254", Rotation="20.000,0.066,0.016", Scale="1,1,1") {
    alphaInputType=Bool
    rotationMode=BMM_Additive
    rotationSpace=BCS_BoneSpace
    scaleMode=BMM_Additive
    scaleSpace=BCS_BoneSpace
    translationMode=BMM_Additive
    translationSpace=BCS_BoneSpace
    alphaBoolBlend {blendInTime=0.5, blendOption=Linear, blendOutTime=0.3}
    alphaScaleBiasClamp {bInterpResult=true}
    boneToModify {boneName=dyn_bck_braids_r_B_02}
    rotation=<rot 20 0.0659 0.016>
    translation=<-0.077 -6.271 0.254>
   }
   n94 = "Transform (Modify) Bone - Bone: dyn_bck_braids_r_C_02"(ComponentPose=@n93, bAlphaBoolEnabled=True, Translation="0,0,0", Rotation="0.011,-20.001,-0.019", Scale="1,1,1") {
    alphaInputType=Bool
    rotationMode=BMM_Additive
    rotationSpace=BCS_BoneSpace
    scaleMode=BMM_Additive
    scaleSpace=BCS_BoneSpace
    translationMode=BMM_Additive
    translationSpace=BCS_BoneSpace
    alphaBoolBlend {blendInTime=0.5, blendOption=Linear, blendOutTime=0.3}
    alphaScaleBiasClamp {bInterpResult=true}
    boneToModify {boneName=dyn_bck_braids_r_C_02}
    rotation=<rot 0.011 -20.001 -0.019>
   }
   n95 = "Transform (Modify) Bone - Bone: dyn_bck_braids_l_A_03"(ComponentPose=@n94, bAlphaBoolEnabled=True, Translation="0,-1,2", Rotation="0,0,0", Scale="1,1,1") {
    alphaInputType=Bool
    rotationMode=BMM_Additive
    rotationSpace=BCS_BoneSpace
    scaleMode=BMM_Additive
    scaleSpace=BCS_BoneSpace
    translationMode=BMM_Additive
    translationSpace=BCS_BoneSpace
    alphaBoolBlend {blendInTime=0.5, blendOption=Linear, blendOutTime=0.3}
    alphaScaleBiasClamp {bInterpResult=true}
    boneToModify {boneName=dyn_bck_braids_l_A_03}
    translation {y=-1, z=2}
   }
   n96 = "Transform (Modify) Bone - Bone: dyn_bck_braids_l_B_03"(ComponentPose=@n95, bAlphaBoolEnabled=True, Translation="0,-1,2", Rotation="0,0,0", Scale="1,1,1") {
    alphaInputType=Bool
    rotationMode=BMM_Additive
    rotationSpace=BCS_BoneSpace
    scaleMode=BMM_Additive
    scaleSpace=BCS_BoneSpace
    translationMode=BMM_Additive
    translationSpace=BCS_BoneSpace
    alphaBoolBlend {blendInTime=0.5, blendOption=Linear, blendOutTime=0.3}
    alphaScaleBiasClamp {bInterpResult=true}
    boneToModify {boneName=dyn_bck_braids_l_B_03}
    translation {y=-1, z=2}
   }
   n97 = "Transform (Modify) Bone - Bone: dyn_bck_braids_l_C_03"(ComponentPose=@n96, bAlphaBoolEnabled=True, Translation="0.000,-1,2", Rotation="0,0,0", Scale="1,1,1") {
    alphaInputType=Bool
    rotationMode=BMM_Additive
    rotationSpace=BCS_BoneSpace
    scaleMode=BMM_Additive
    scaleSpace=BCS_BoneSpace
    translationMode=BMM_Additive
    translationSpace=BCS_BoneSpace
    alphaBoolBlend {blendInTime=0.5, blendOption=Linear, blendOutTime=0.3}
    alphaScaleBiasClamp {bInterpResult=true}
    boneToModify {boneName=dyn_bck_braids_l_C_03}
    translation {y=-1, z=2}
   }
   n98 = "Transform (Modify) Bone - Bone: dyn_bck_braids_l_A_02"(ComponentPose=@n97, bAlphaBoolEnabled=True, Translation="0,0,0", Rotation="0.100,9.998,-0.045", Scale="1,1,1") {
    alphaInputType=Bool
    rotationMode=BMM_Additive
    rotationSpace=BCS_BoneSpace
    scaleMode=BMM_Additive
    scaleSpace=BCS_BoneSpace
    translationMode=BMM_Additive
    translationSpace=BCS_BoneSpace
    alphaBoolBlend {blendInTime=0.5, blendOption=Linear, blendOutTime=0.3}
    alphaScaleBiasClamp {bInterpResult=true}
    boneToModify {boneName=dyn_bck_braids_l_A_02}
    rotation=<rot 0.1 9.998 -0.045>
   }
   n99 = "Transform (Modify) Bone - Bone: dyn_bck_braids_l_B_02"(ComponentPose=@n98, bAlphaBoolEnabled=True, Translation="0,0,0", Rotation="10.004,-10.198,-0.042", Scale="1,1,1") {
    alphaInputType=Bool
    rotationMode=BMM_Additive
    rotationSpace=BCS_BoneSpace
    scaleMode=BMM_Additive
    scaleSpace=BCS_BoneSpace
    translationMode=BMM_Additive
    translationSpace=BCS_BoneSpace
    alphaBoolBlend {blendInTime=0.5, blendOption=Linear, blendOutTime=0.3}
    alphaScaleBiasClamp {bInterpResult=true}
    boneToModify {boneName=dyn_bck_braids_l_B_02}
    rotation=<rot 10.004 -10.198 -0.042>
   }
   n100 = "Transform (Modify) Bone - Bone: dyn_bck_braids_l_C_02"(ComponentPose=@n99, bAlphaBoolEnabled=True, Translation="0,0,0", Rotation="9.813,20.209,1.772", Scale="1,1,1") {
    alphaInputType=Bool
    rotationMode=BMM_Additive
    rotationSpace=BCS_BoneSpace
    scaleMode=BMM_Additive
    scaleSpace=BCS_BoneSpace
    translationMode=BMM_Additive
    translationSpace=BCS_BoneSpace
    alphaBoolBlend {blendInTime=0.5, blendOption=Linear, blendOutTime=0.3}
    alphaScaleBiasClamp {bInterpResult=true}
    boneToModify {boneName=dyn_bck_braids_l_C_02}
    rotation=<rot 9.813 20.209 1.772>
   }
   n101 = "Constraint - dyn_fr_braid_r_04 (PPP)"!(ComponentPose=@n100, Alpha=0.95, ConstraintWeights_0=0.7319, ConstraintWeights_1:float=0.2601, ConstraintWeights_2:float=0.008) -> (Pose) {
    alphaBoolBlend {blendOption=Linear}
    boneToModify {boneName=dyn_fr_braid_r_04}
    constraintSetup [{targetBone {boneName=spine_04}}, {targetBone {boneName=spine_05}}, {targetBone {boneName=clavicle_r}}]
    constraintWeights [1,1,1]
   }
   n102 = "Constraint - dyn_fr_braid_l_04 (PPP)"!(ComponentPose=@n101, Alpha=0.95, ConstraintWeights_0=0.7118, ConstraintWeights_1:float=0.275, ConstraintWeights_2:float=0.0132) -> (Pose) {
    alphaBoolBlend {blendOption=Linear}
    boneToModify {boneName=dyn_fr_braid_l_04}
    constraintSetup [{targetBone {boneName=spine_04}}, {targetBone {boneName=spine_05}}, {targetBone {boneName=clavicle_l}}]
    constraintWeights [1,1,1]
   }
   Output Pose(Result=Blend Poses by bool(
     BlendPose_0=Component To Local(ComponentPose=@n54)
     BlendPose_1=Component To Local(ComponentPose=@n102)
     bActiveValue=Get bRigidBodyEnabled()
    ) {blendType=Linear}) {layerGroup=None}
   Save cached pose 'NullPose'(Pose=Copy Pose From Mesh() {bUseAttachedParent=true}) {cacheName=NullPose}
   Two Way Blend(Alpha=0) {blendNode {alphaBoolBlend {blendOption=Linear}}}
  ]}
 EventGraph {nodes [
   Event BlueprintUpdateAnimation()
   > n1 = Sequence!(execute) -> (then_0, then_1, then_2:Exec)
   n2 = Gravity Override State Function DEPRECATED(
    execute=@n1.then_1
    CharacterPartAnimInstance=Self-Reference()
    GameTimeInSeconds=GetGameTimeInSeconds()
    GravityOverrideParams=Get Gravity Override Params()
   )
   > Set GravityOverride(GravityOverride=@n2)
   n3 = Branch(execute=@n1.then_0, Condition=CheckForBackbling())
   > Set hasBackpack(hasBackpack:double=1) -> (Output_Get:double)
   > n4 = Set WearingBackbling(WearingBackbling:bool=true) -> (Output_Get:bool)
   > Set backblingfrontend(backblingfrontend:bool=AND Boolean(A=IsCosmeticObjectDisplayedInFrontendContext!(CosmeticObject:Object Reference) -> (ReturnValue:bool), B=@n4)) -> (Output_Get:bool)
   > n5 = Branch(Condition=Get bIsSkydiveFloating())
   > n6 = Set hasBackpack(hasBackpack=0.5)
   Set skydiveAlpha(execute=[@n20.then, @n6.then], skydiveAlpha:double=1) -> (Output_Get:double)
   > n7 = Set normalAlpha(normalAlpha:double=0) -> (Output_Get:double)
   n8 = GetOwnerSkeletalMeshComponent()
   n9 = BreakRotator(InRot=BreakTransform(InTransform=GetDeltaTransformFromRefPose(self=@n8, BoneName=upperarm_r, BaseName=spine_05)).Rotation)
   n10 = Set upperarm_R_rot(execute=[@n7.then, @n19.then], upperarm_R_rot:double=MapRangeClamped(Value="Absolute (Float)"(A=@n9.Pitch), InRangeA=100, InRangeB=-100)) -> (Output_Get:double)
   n11 = Get bIsTargeting()
   n12 = "To Float (Boolean)"(InBool=@n11)
   n13 = "To Float (Boolean)"(InBool=Get bIsCrouching())
   n14 = BreakTransform(InTransform=GetDeltaTransformFromRefPose(self=GetOwnerSkeletalMeshComponent(), BoneName=head, BaseName=spine_05))
   n15 = MapRangeClamped(Value="Absolute (Float)"(A=BreakRotator(InRot=@n14.Rotation).Pitch), InRangeA=0, InRangeB=50)
   > Set crouching_aimup_alpha(crouching_aimup_alpha:double=float * float(A=float * float(A=@n15, B=@n13), B=@n12) {is=title K2Node_PromotableOperator}) -> (Output_Get:double)
   > n16 = Branch(Condition=@n11)
   > n17 = Set targeting_alpha(targeting_alpha:double=0.4) -> (Output_Get:double)
   Set moving_alpha(execute=[@n17.then, @n18.then], moving_alpha:double=MapRangeClamped(Value=Get PawnSpeed2D(), InRangeA=0, InRangeB=500)) -> (Output_Get:double)
   > Set headUp_alpha(headUp_alpha:double=float * float(A=@n12, B=@n15)) -> (Output_Get:double)
   n18 = Set targeting_alpha(execute=@n16.else, targeting_alpha=0)
   Set skydiveAlpha(execute=[@n5.else, @n20.else], skydiveAlpha=0)
   > n19 = Set normalAlpha(normalAlpha=1)
   Set hasBackpack(execute=@n3.else, hasBackpack=0.3)
   > Set WearingBackbling(WearingBackbling=false)
   > n20 = Branch(Condition=Get bIsSkydiveFloating())
   n21 = Branch(execute=@n1.then_2, Condition=OR Boolean(A=Get bIsFalling(), B=Get bIsJumping()))
   > Set JumpingFalling(JumpingFalling:bool=true) -> (Output_Get:bool)
   Set JumpingFalling(execute=@n21.else, JumpingFalling=false)
   float * float(A=@n10, B=@n15)
   Set targeting_alpha(targeting_alpha=0)
  ]}
]
variables [
 backblingfrontend {type=boolean}
 crouching_aimup_alpha {type=number}
 Gravity Override Params {type=GravityOverrideParamsStruct}
 GravityOverride {type=Vector}
 hasBackpack {type=number}
 headUp_alpha {type=number}
 JumpingFalling {type=boolean}
 moving_alpha {type=number}
 normalAlpha {type=number}
 skydiveAlpha {type=number}
 targeting_alpha {type=number}
 upperarm_R_rot {type=number}
 WearingBackbling {type=boolean}
]
